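# Masked subsurface material for Niagara ribbon and mesh particles, drawing goo trails for the Victory Highway emote VFX.
# A powered trail texture on tiled, panned or center-scaled UV1 combines with panning eataway noise, eroded by particle alpha.
# Trail normals blend with flattened noise normals and feed a Fresnel-tinted subsurface color.
# Particle color and dynamic parameters supply emissive scaling and U offset.
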Material /BRCosmetics/Animation/Game/MainPlayer/Emotes/VictoryHighway/VFX/M_GooDynamics_Ribbon
graph {
 properties {
  bUsedWithNiagaraMeshParticles=true
  bUsedWithNiagaraRibbons=true
  blendMode=BLEND_Masked
  pixelDepthOffsetMode=PDOM_Legacy
  shadingModel=MSM_Subsurface
  usedShadingModels=Subsurface
 }
 nodes [
  n1 = TextureSampleParameter2D(UVs=NamedRerouteUsage()) {parameterName=EatawayNoise, texture=/BRCosmetics/Athena/Effects/Particles/Skydiving/NS_Hightower_Date/T_AtmosphericCloudNoise_Vertical}
  n2 = ChannelMaskParameter(None=@n1.RGBA) {maskChannel=Green, parameterName=Eataway Channel, defaultValue=<rgba 0 1 0 0>}
  n3 = NamedRerouteUsage()
  n4 = TextureSampleParameter2D(UVs=@n3) {parameterName=Trail Texture, samplerType=SAMPLERTYPE_LinearColor, texture=/Game/Effects/Fort_Effects/Textures/Liquid/T_GooRibbonBasic}
  n5 = TextureSampleParameter2D(UVs=@n3) {
   parameterName=Trail Texture
   samplerSource=SSM_Clamp_WorldGroupSettings
   samplerType=SAMPLERTYPE_LinearColor
   texture=&T_GooRibbonBasic
  }
  n6 = ChannelMaskParameter(None=LinearInterpolate(A=@n4.RGBA, B=@n5.RGBA, Alpha=ScalarParameter() {parameterName=Trail Clamped UVs})) {maskChannel=Green, parameterName=Trail Channel, defaultValue=<rgba 0 1 0 0>}
  n7 = Power(Base=@n6.R, Exp=ScalarParameter() {defaultValue=4, parameterName=TrailPower})
  n8 = Add(A=@n2.R, B=@n7)
  n9 = Saturate(None=ParticleColor().A)
  Multiply(A=NamedRerouteUsage(), B=LinearInterpolate(A=Multiply(A=@n8, B=@n6.R), B=@n7, Alpha=Multiply(A=@n9, B=@n9)))
  Multiply(A=ParticleColor().RGB, B=DynamicParameter() {defaultValue=<rgba 1 1 1 0>, paramNames [UOffset,Param2,Param3,Emissive]}.Emissive)
  n10 = ScalarParameter() {defaultValue=1, parameterName=SubsurfaceExp}
  n11 = NamedRerouteUsage()
  n12 = TextureSampleParameter2D(UVs=@n11) {parameterName=Normal Texture, samplerType=SAMPLERTYPE_Normal, texture=/Game/Effects/Fort_Effects/Textures/Liquid/T_GooRibbonBasic_NM}
  n13 = TextureSampleParameter2D(UVs=@n11) {
   parameterName=Normal Texture
   samplerSource=SSM_Clamp_WorldGroupSettings
   samplerType=SAMPLERTYPE_Normal
   texture=&T_GooRibbonBasic_NM
  }
  n14 = VectorParameter() {parameterName=EatawayNormal TilePan, defaultValue=<rgba 1 1 0 0>}
  n15 = LinearInterpolate(
   A=NamedRerouteUsage()
   B=MaterialFunctionCall!(UV=TextureCoordinate() {coordinateIndex=1}, Tile U=@n14.R, Tile V=@n14.R, Pan U=@n14.R, Pan V=@n14.R, Offset U, Offset V, Time) {materialFunction=/Game/Effects/Fort_Effects/Materials/_Functions/MF_PanTileOffset}.UV's
   Alpha=ScalarParameter() {parameterName=Eataway Noise Normal Unique UVs}
  )
  n16 = MaterialFunctionCall(
   BaseNormal=LinearInterpolate(A=@n12.RGB, B=@n13.RGB, Alpha=ScalarParameter() {parameterName=Trail Clamped UVs})
   AdditionalNormal=MaterialFunctionCall(
    Normal=TextureSampleParameter2D(UVs=@n15) {parameterName=Eataway Noise Normal, samplerType=SAMPLERTYPE_Normal, texture=/Game/Global/Textures/Default/Noise/T_SwirlNoise_N}.RGB
    Flatness=ScalarParameter() {defaultValue=0.8, parameterName=Eataway Normal Flatness}
   ) {materialFunction=&FlattenNormal}.Result
  ) {materialFunction=&BlendAngleCorrectedNormals}
  n17 = LinearInterpolate(A=@n16.Result, B=Multiply(A=@n16.Result, B=Constant3Vector() {constant=<rgba 1 -1 1 0>}), Alpha=ScalarParameter() {parameterName=FlipNormal})
  Multiply(
   A=@n6.R
   B=Multiply(
    A=VectorParameter() {parameterName=SubsurfaceColor, defaultValue=<rgba 1 0 0 0>}.RGB
    B=Fresnel(ExponentIn=@n10, Normal=@n17) {baseReflectFraction=0}
   )
  )
  n18 = LinearInterpolate(Alpha=ParticleColor().A) {constA=-1}
  NamedRerouteDeclaration(None=Saturate(None=Add(A=@n18, B=@n2.R))) {name=Eataway, nodeColor=<rgba 0.5937 1 0 1>}
  n19 = TextureCoordinate() {coordinateIndex=1}
  n20 = ScalarParameter() {defaultValue=1, parameterName=Tile U}
  n21 = ScalarParameter() {parameterName=Pan U}
  n22 = DynamicParameter() {defaultValue=<rgba 1 1 1 0>, paramNames [UOffset,Param2,Param3,Emissive]}
  n23 = TextureCoordinate() {coordinateIndex=1}
  n24 = ScalarParameter() {defaultValue=1, parameterName=Scale By Center Amount}
  NamedRerouteDeclaration(None=LinearInterpolate(
    A=MaterialFunctionCall(UV=@n19, Tile U=@n20, Pan U=@n21, Offset U=@n22.UOffset) {materialFunction=&MF_PanTileOffset}.UV's
    B=MaterialFunctionCall(UVs=@n23, Texture Scale=@n24) {materialFunction=&ScaleUVsByCenter}.UVs
    Alpha=ScalarParameter() {parameterName=Scale by Center}
   )) {name="Trail UVs (UV1)", nodeColor=<rgba 0 1 0.5078 1>}
  NamedRerouteDeclaration(None=MaterialFunctionCall(
    UV=TextureCoordinate() {coordinateIndex=1}
    Tile U=ScalarParameter() {defaultValue=1, parameterName=Eataway Tile U}
    Tile V=ScalarParameter() {defaultValue=1, parameterName=Eataway Tile V}
    Pan U=ScalarParameter() {defaultValue=-0.75, parameterName=Eataway Pan U}
    Offset U=DynamicParameter() {defaultValue=<rgba 1 1 1 0>, paramNames [UOffset,Param2,Param3,Emissive]}.UOffset
   ) {materialFunction=&MF_PanTileOffset}.UV's) {name=Eataway UVs, nodeColor=<rgba 1 0.6328 0 1>}
  ScalarParameter() {parameterName=M}
  ScalarParameter() {defaultValue=0.2, parameterName=R}
 ]
}
