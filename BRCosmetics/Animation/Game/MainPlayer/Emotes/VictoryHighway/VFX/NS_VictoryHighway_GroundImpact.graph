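# Niagara ground impact effect for the VictoryHighway player emote, throwing up fire, dust and rubble.
# Burst-driven emitters layer embers, flame streaks, dust, glowing crack decals, GPU debris and lava splash; BlowingParticles and FlameBeam are disabled.
# User.DesertBoulder switches dust tint, while User.bIsFrontEndPreview adjusts dust drift and crack lifetime and fade.
# Platform quality sets gate debris visibility and collision, both also enabled by front-end preview.
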NiagaraSystem /BRCosmetics/Animation/Game/MainPlayer/Emotes/VictoryHighway/VFX/NS_VictoryHighway_GroundImpact
dependencies {
 usedDataInterfaces [
  &NiagaraDataInterfaceMeshRendererInfo
  &NiagaraDataInterfaceCurve
  &NiagaraDataInterfacePlatformSet
  &NiagaraDataInterfaceSkeletalMesh
  &NiagaraDataInterfaceColorCurve
  &NiagaraDataInterfaceSpriteRendererInfo
  &NiagaraDataInterfaceVector2DCurve
  &NiagaraDataInterfaceVectorCurve
  &NiagaraDataInterfaceCamera
 ]
 usedDynamicInputs [
  &SimulationPosition
  &UniformRangedVector2D
  &MakeFloatFromVector2d
  &MakeFloatFromVector
 ]
 usedModules [
  &self:SystemSpawnScript.SystemScriptSource.SystemScriptGraph.NiagaraNodeAssignment_0.SetVariables_#1
  &SystemState
 ]
}
emitters {
 BlowingParticles {
  bEnabled=false
  emitterUpdateScript [
   &EmitterState(Life Cycle Mode=::NewEnumerator1, Loop Behavior=::NewEnumerator1, Loop Duration=2, UseLoopDelay=false)
   &SpawnBurst_Instantaneous(Use Spawn Probability=false, UseLoopCountLimit=false)
   /SetVariables(Emitter.float:NiagaraFloat=&FloatFromCurve(FloatCurve=DataInterface() {propertyValues {Curve {keys [<at 0 1>,<at 1 0>]}}}, CurveIndex=$Emitter.NormalizedLoopAge, Scale Curve=0.2)) {isSetParameters=true}
   /SetVariables(Emitter.Linear Color:LinearColor=&FNTimeOfDayColor(
     Night Color=<rgba 1 0.2689 0.3071 1>
     Morning Color=<rgba 1 0.2423 0.5265 1>
     Day Color=<rgba 10 3.1666 0.3538 1>
     Evening Color=<rgba 1 0.2979 0.4227 1>
    )) {isSetParameters=true}
  ]
  particleSpawnScript [&InitializeParticle(
    Lifetime Min=2
    Lifetime Mode=::NewEnumerator1
    Mass Max=1.5
    Mass Mode=::NewEnumerator3
    Sprite Rotation Mode=::NewEnumerator2
    Sprite Size Mode=::NewEnumerator1
    Uniform Sprite Size Max=8
    Uniform Sprite Size Min=4
    UsePositionOffset=false
    Write Material Random=true
   )]
  particleUpdateScript [
   &ParticleState()
   /SetVariables(Particles.DecalSize:Vector3f=&Multiply_Vector(
     A=<20 31.25 31.25>
     B=&VectorFromFloat(Value=&FloatFromCurve(FloatCurve=DataInterface() {propertyValues {Curve {keys [
           {arriveTangent=4.1785, interpMode=RCIM_Cubic, leaveTangent=4.1785, tangentMode=RCTM_User, time=0, value=0.5}
           {arriveTangent=1.1179, interpMode=RCIM_Cubic, leaveTangent=1.1179, tangentMode=RCTM_User, time=0.01, value=0.8}
           {arriveTangent=0.0224, interpMode=RCIM_Cubic, leaveTangent=0.0224, tangentMode=RCTM_User, time=0.05, value=1}
          ]}}}))
    )) {isSetParameters=true}
   &ScaleColor(
    Scale Alpha=&FloatFromCurve(FloatCurve=DataInterface() {propertyValues {Curve {keys [<at 0 1>,<at 1 0>]}}})
    ScaleA=true
    ScaleRGB=true
   )
  ]
  rendererClasses [&NiagaraDecalRendererProperties]
  renderers [{rendererClass=&NiagaraDecalRendererProperties, rendererIndex=0}]
 }
 Embers {
  emitterUpdateScript [
   &EmitterState()
   &SpawnRate(
    SpawnRate=&FloatFromCurve(FloatCurve=DataInterface() {propertyValues {Curve {keys [<at 0.1 1>,<at 0.75 0>]}}}, CurveIndex=$Emitter.Age, Scale Curve=20)
    Use Spawn Probability=false
   )
   &SpawnBurst_Instantaneous(Spawn Count=30, Use Spawn Probability=false, UseLoopCountLimit=false)
  ]
  particleSpawnScript [
   &InitializeParticle(
    Color Mode=::NewEnumerator0
    Lifetime Max=0.75
    Lifetime Min=0.5
    Lifetime Mode=::NewEnumerator1
    Sprite Size Mode=::NewEnumerator1
    Uniform Sprite Size Max=4
    Uniform Sprite Size Min=1
    UsePositionOffset=false
    Write Material Random=true
   )
   &ShapeLocation(
    MaskSpawnGroup=false
    Offset Coordinate Space=::Local
    Offset Mode=::NewEnumerator0
    Override Seed=false
    Sphere Radius=0
    UseOwnerScale=true
   )
   &AddVelocity(
    Cone Angle=50
    Cone Axis=<0 0 1>
    Module.UseConeFalloff=false
    Module.UseVelDistribution=false
    Override Seed=false
    Velocity Mode=::NewEnumerator2
    Velocity Speed=&RandomRangeFloat(Minimum=222, Maximum=999)
   )
   &ApplyInitialForces()
  ]
  particleUpdateScript [
   &ParticleState()
   &Drag(Drag=4, Use Linear Drag=true, Use Rotational Drag=true)
   &GravityForce(Gravity=<0 0 -444>)
   &SolveForcesAndVelocity(Clamp Velocity=false, Limit Acceleration=false)
   &ScaleColor(
    Scale RGB=&MakeVectorFromLinearColorRGB(Color=&ColorFromCurve(ColorCurve=DataInterface() {propertyValues {
        AlphaCurve {keys [<at 0 1 RCIM_Cubic>]}
        BlueCurve {keys [<at 0 0 RCIM_Cubic>,<at 0.333 0>,<at 1 0 RCIM_Cubic>]}
        GreenCurve {keys [<at 0 3.3 RCIM_Cubic>, <at 0.333 0.3>, {interpMode=RCIM_Cubic, leaveTangent=-0.4498, time=1, value=0}]}
        RedCurve {keys [<at 0 22 RCIM_Cubic>, <at 0.333 6>, {interpMode=RCIM_Cubic, leaveTangent=-8.9955, time=1, value=1}]}
       }}))
    ScaleA=true
    ScaleRGB=true
   )
   &ScaleSpriteSizeBySpeed(Max Scale Factor=<1 3>, Min Scale Factor=<1 -3>, Velocity Threshold=222)
  ]
  rendererClasses [&NiagaraSpriteRendererProperties]
  renderers [{rendererClass=&NiagaraSpriteRendererProperties}]
 }
 FlameBeam {
  bEnabled=false
  emitterUpdateScript [
   &EmitterState()
   &SpawnRate(
    SpawnRate=&FloatFromCurve(FloatCurve=DataInterface() {propertyValues {Curve {keys [<at 0 1>,<at 0.47 0>]}}}, CurveIndex=$System.Age, Scale Curve=23)
    Use Spawn Probability=false
   )
  ]
  particleSpawnScript [
   &InitializeParticle(
    Color Maximum=<rgba 1 0.6294 0.6199 1>
    Color Minimum=<rgba 1 0.9684 0.9781 1>
    Color Mode=::NewEnumerator2
    Lifetime=&RandomRangeFloat(Minimum=0.6, Maximum=0.8)
    Sprite Rotation Mode=::NewEnumerator2
    Sprite Size Mode=::NewEnumerator3
    Uniform Sprite Size=&Multiply_Float(A=&RandomRangeFloat(Minimum=170, Maximum=220), B=$System.FX_Scale)
    UsePositionOffset=false
    Write Material Random=true
   )
   &ShapeLocation(
    MaskSpawnGroup=false
    Offset=&RotateVectorByQuaternion(Vector To Rotate=<0 0 0>, Quaternion=$Engine.Owner.Rotation)
    Offset Mode=::NewEnumerator0
    Override Seed=false
    Ring Radius=&Multiply_Float(A=4, B=$System.FX_Scale)
    Rotation Mode=::NewEnumerator1
    Shape Primitive=::NewEnumerator4
    UseOwnerScale=true
   )
   &AddVelocity(Velocity=&Multiply_VectorByFloat(Vector=$Particles.ShapeLocation.ShapeTangent, Float=10))
   &AddVelocity(Velocity=<0 0 150>, Velocity Speed Scale=&RandomRangeFloat(Minimum=0.9, Maximum=1)) {module=AddVelocity001}
   &InheritVelocity(Inherited Velocity Speed Limit=3000, Speed Limit=true, UseSpeedThreshold=false)
   &CameraOffset(Camera Offset Amount=&RandomRangeFloat(Minimum=15, Maximum=30))
  ]
  particleUpdateScript [
   &self:ScratchModule_01() {module=ScratchModule_01}
   &Drag(Use Linear Drag=true, Use Rotational Drag=true)
   &SolveForcesAndVelocity(Clamp Velocity=false, Limit Acceleration=false)
   &ParticleState()
   &SpriteRotationRate(Rotation Rate=&FloatFromCurve(
     FloatCurve=DataInterface() {propertyValues {Curve {keys [<at 0 1>,<at 1 -1>]}}}
     CurveIndex=$Output.ScratchModule_01.SpriteRotationDirectionMask
     Scale Curve=-300
    ))
   &ScaleSpriteSize(Uniform Curve Sprite Scale=DataInterface() {propertyValues {Curve {keys [
        {arriveTangent=-0.1234, interpMode=RCIM_Cubic, leaveTangent=-0.1234, tangentMode=RCTM_User, time=0, value=0.3}
        {interpMode=RCIM_Cubic, tangentMode=RCTM_User, time=0.7, value=0.8}
       ]}}})
   &ScaleColor(
    Scale Alpha=&FloatFromCurve(
     FloatCurve=DataInterface() {propertyValues {Curve {keys [
         {interpMode=RCIM_Cubic, tangentMode=RCTM_User, time=0, value=0}
         {interpMode=RCIM_Cubic, tangentMode=RCTM_User, time=0.3, value=1}
         {interpMode=RCIM_Cubic, tangentMode=RCTM_User, time=1, value=0}
        ]}}}
     Scale Curve=&Multiply_Float(A=0.1, B=1)
    )
    Scale RGB=&MakeVectorFromLinearColorRGB(Color=&ColorFromCurve(ColorCurve=DataInterface() {propertyValues {
        AlphaCurve {keys [<at 0 1 RCIM_Cubic>]}
        BlueCurve {keys [<at 0 0.24 RCIM_Cubic>,<at 0.4931 0>,<at 1 0 RCIM_Cubic>]}
        GreenCurve {keys [<at 0 0.8372 RCIM_Cubic>,<at 0.4931 0.6752>,<at 1 0 RCIM_Cubic>]}
        RedCurve {keys [<at 0 10 RCIM_Cubic>,<at 0.4931 2>,<at 1 1 RCIM_Cubic>]}
       }}))
    ScaleA=true
    ScaleRGB=true
   )
   &SubUVAnimation(
    "Sprite Renderer "=DataInterface() {propertyValues {SpriteRenderer=&self:FlameBeam_0.NiagaraSpriteRendererProperties_0}}
    Start Frame Offset=4
    UseEndFrame=false
    UseStartFrame=false
   )
   &DynamicMaterialParameters(
    Index 0 Param 2=&FloatFromCurve(
     FloatCurve=DataInterface() {propertyValues {Curve {keys [
         {arriveTangent=-2.5672, interpMode=RCIM_Cubic, leaveTangent=-2.5672, tangentMode=RCTM_User, time=0, value=1}
         {interpMode=RCIM_Cubic, tangentMode=RCTM_User, time=1, value=0}
        ]}}}
     Scale Curve=&RandomRangeFloat(Minimum=0.1, Maximum=0.2)
    )
    Module.Index0Param0Write=false
    Module.Index0Param1Write=true
    Module.Index0Param2Write=false
    Module.Index0Param3Write=false
   )
  ]
  rendererClasses [&NiagaraSpriteRendererProperties]
  renderers [{rendererClass=&NiagaraSpriteRendererProperties}]
 }
 Flame_Streaks {
  emitterUpdateScript [
   &EmitterState()
   &SpawnBurst_Instantaneous(
    Spawn Count=&RandomRangeInt(Minimum=12, Maximum=14)
    Use Spawn Probability=false
    UseLoopCountLimit=false
   )
  ]
  particleSpawnScript [
   &PartitionParticles(Particles Per Partition=7, Partition Mode:ENiagara_PartitionMode_v2=::NewEnumerator1)
   &InitializeParticle(
    Lifetime=&RandomRangeFloat(Minimum=0.12, Maximum=0.14)
    Sprite Size=&Multiply_Vector2D_ByFloat(
     Float=&RandomRangeFloat(Minimum=0.7, Maximum=1)
     Vector2D=&RandomRangeVector2D(Minimum=<8 33>, Maximum=<13 40>)
    )
    Sprite Size Mode=::NewEnumerator4
    UsePositionOffset=false
    Write Material Random=true
   )
   &SubUVAnimation(
    End Frame Range Override=3
    Evaluation Type=::NewEnumerator0
    "Sprite Renderer "=DataInterface() {propertyValues {SpriteRenderer=&self:Flame_Streaks_1.NiagaraSpriteRendererProperties_0}}
    SubUV Animation Mode:ENiagara_SubUVLookupModeV2=::NewEnumerator2
    UseEndFrame=true
    UseStartFrame=true
   )
   &ShapeLocation(
    Handle Radius=1
    Large Radius=&MergeHelper_FloatSwitch(0=10, 1=20, 2=?, 3=?, Index=$Particles.PartitionParticles.Partition, Num=::NewEnumerator0)
    MaskSpawnGroup=false
    Offset Mode=::NewEnumerator0
    Override Seed=false
    Shape Primitive=::NewEnumerator3
    UseOwnerScale=true
   )
   &AddVelocity(
    Module.Use Falloff Distance=false
    Origin Offset=&MakeVector(X=0, Y=0, Z=&RandomRangeFloat(Minimum=0, Maximum=-5))
    Override Seed=false
    Velocity Mode=::NewEnumerator1
    Velocity Speed=&RandomRangeFloat(Minimum=300, Maximum=400)
   )
   &CameraOffset(Camera Offset Amount=10)
  ]
  particleUpdateScript [
   &Drag(
    Drag=&Multiply_Float(
     A=&RandomRangeFloat(Minimum=10, Maximum=20)
     B=&FloatFromCurve(FloatCurve=DataInterface() {propertyValues {Curve {keys [<at 0 0>,<at 0.3 1>]}}})
    )
    Use Linear Drag=true
    Use Rotational Drag=true
   )
   &SolveForcesAndVelocity(Clamp Velocity=false, Limit Acceleration=false)
   &ParticleState()
   &ScaleSpriteSize(
    Non-Uniform Curve Sprite Scale=DataInterface() {propertyValues {
      XCurve {keys [{interpMode=RCIM_Cubic, tangentMode=RCTM_User, time=0, value=1}, {interpMode=RCIM_Cubic, tangentMode=RCTM_User, time=1, value=1}]}
      YCurve {keys [{interpMode=RCIM_Cubic, tangentMode=RCTM_User, time=0.35, value=1.3}, {interpMode=RCIM_Cubic, tangentMode=RCTM_User, time=1, value=0.7}]}
     }}
    Scale Sprite Size Mode=::NewEnumerator3
   )
   &ScaleColor(
    Scale Alpha=&FloatFromCurve(FloatCurve=DataInterface() {propertyValues {Curve {
        defaultValue=1
        keys [
         {tangentMode=RCTM_User, time=0, value=1}
         {arriveTangent=-0.0405, interpMode=RCIM_Cubic, leaveTangent=-0.0405, tangentMode=RCTM_User, time=0.7009, value=1}
         {arriveTangent=-8.0089, interpMode=RCIM_Cubic, leaveTangent=-8.0089, tangentMode=RCTM_User, time=1, value=0}
        ]
       }}})
    Scale RGB=&MakeVectorFromLinearColorRGB(Color=&MergeHelper_ColorSwitch(
      0=&ColorFromCurve(ColorCurve=DataInterface() {propertyValues {
         AlphaCurve {keys [<at 0 1 RCIM_Cubic>]}
         BlueCurve {keys [<at 0.0898 1.61 RCIM_Cubic>,<at 0.3223 0.0037>,<at 1 0 RCIM_Cubic>]}
         GreenCurve {keys [<at 0.0898 4.7542 RCIM_Cubic>,<at 0.3223 0.8205>,<at 1 0 RCIM_Cubic>]}
         RedCurve {keys [<at 0.0898 7 RCIM_Cubic>,<at 0.3223 3>,<at 1 1 RCIM_Cubic>]}
        }})
      1=&ColorFromCurve(ColorCurve=DataInterface() {propertyValues {
         AlphaCurve {keys [<at 0 1 RCIM_Cubic>]}
         BlueCurve {keys [<at 0.0228 0 RCIM_Cubic>,<at 0.2447 0.009>,<at 1 0.0074 RCIM_Cubic>]}
         GreenCurve {keys [<at 0.0228 0.0037 RCIM_Cubic>,<at 0.2447 0.0154>,<at 1 0.0074 RCIM_Cubic>]}
         RedCurve {keys [<at 0.0228 0.488 RCIM_Cubic>,<at 0.2447 0.1753>,<at 1 0.0179 RCIM_Cubic>]}
        }})
      2=?
      3=?
      Index=$Particles.Initial.PartitionParticles.Partition
      Num=::NewEnumerator0
     ))
    ScaleA=true
    ScaleRGB=true
   )
  ]
  rendererClasses [&NiagaraSpriteRendererProperties]
  renderers [{rendererClass=&NiagaraSpriteRendererProperties}]
 }
 GroundDust {
  emitterUpdateScript [
   &EmitterState()
   &SpawnRate(SpawnRate=50, Use Spawn Probability=false) {enabled=false}
   &SpawnBurst_Instantaneous(Spawn Count=14, Use Spawn Probability=false, UseLoopCountLimit=false)
   /SetVariables(Emitter.SizeRad=25) {isSetParameters=true}
  ]
  particleSpawnScript [
   &PartitionParticles(
    Particles Per Partition=14
    Partition Count Mode:ENiagara_InfiniteLoopDuration=::NewEnumerator1
    Partition Mode:ENiagara_PartitionMode_v2=::NewEnumerator1
   )
   /SetVariables(Particles.DustColor:LinearColor=&LerpLinearColors(
     StartColor=&RandomRangeLinearColor(Minimum=<rgba 0.1146 0.1041 0.0995 0.75>, Maximum=<rgba 0.3229 0.2978 0.2778 0.75>)
     EndColor=&RandomRangeLinearColor(Minimum=<rgba 0.2083 0.0717 0.032 0.75>, Maximum=<rgba 0.4323 0.2135 0.144 0.75>)
     LerpFactor=&MakeCustomFloatFromBool(Bool=$User.DesertBoulder, True Float=1, False Float=0)
    )) {isSetParameters=true}
   /SetVariables(Particles.SizeMax=45, Particles.SizeMin=95) {isSetParameters=true}
   /SetVariables(Particles.SizeCurve:NiagaraFloat=&Lerp_Float(
     A=$Particles.SizeMin
     B=$Particles.SizeMax
     Alpha=&FloatFromCurve(
      FloatCurve=DataInterface() {propertyValues {Curve {keys [
          {interpMode=RCIM_Cubic, tangentMode=RCTM_Break, time=0, value=0}
          {arriveTangentWeight=0.9217, interpMode=RCIM_Cubic, tangentMode=RCTM_Break, time=1, value=1}
         ]}}}
      CurveIndex=&RandomRangeFloat(Minimum=0, Maximum=1)
     )
    )) {isSetParameters=true}
   /SetVariables(Particles.InitSpriteRotRate=&UniformRangedFloat(Minimum=-1, Maximum=1)) {isSetParameters=true}
   &InitializeParticle(
    Color=$Particles.DustColor
    Lifetime=&Lerp_Float(
     A=0.6
     B=1
     Alpha=&FloatFromCurve(
      FloatCurve=DataInterface() {propertyValues {Curve {keys [
          {interpMode=RCIM_Cubic, tangentMode=RCTM_User, time=0, value=0}
          {arriveTangent=2.8717, interpMode=RCIM_Cubic, leaveTangent=2.8717, tangentMode=RCTM_User, time=1, value=1}
         ]}}}
      CurveIndex=$Particles.PartitionParticles.Normalized Index In Partition
     )
    )
    Sprite Rotation Mode=::NewEnumerator2
    Sprite Size Mode=::NewEnumerator3
    Uniform Sprite Size=$Particles.SizeCurve
    UsePositionOffset=false
    Write Material Random=true
   )
   &ShapeLocation(
    Cylinder Distribution=::NewEnumerator2
    Cylinder Height=0
    Cylinder Height Midpoint=0
    Cylinder Radius=10
    MaskSpawnGroup=false
    Offset Mode=::NewEnumerator0
    Shape Primitive=::NewEnumerator1
    U Position=$Particles.PartitionParticles.Normalized Index In Partition
    UseOwnerScale=true
   )
   &AddVelocity(
    Module.Use Falloff Distance=false
    Origin Offset=<0 0 -2.5>
    Override Seed=false
    Velocity Mode=::NewEnumerator1
    Velocity Speed=&RandomRangeFloat(Minimum=108.5645, Maximum=176.389)
   )
   /SetVariables(Particles.Velocity=&Multiply_Vector(A=$Particles.Velocity, B=<1 1 0.5>)) {isSetParameters=true}
   &CameraOffset(Camera Offset Amount=&RandomRangeFloat(Minimum=15, Maximum=25))
   &ApplyInitialForces(Apply Force to Position=true)
   &AddVelocity(
    Velocity=<0 50 0>
    Velocity Speed Scale=&MakeCustomFloatFromBool(Bool=$User.bIsFrontEndPreview, True Float=1, False Float=0)
   ) {module=AddVelocity001}
  ]
  particleUpdateScript [
   &ParticleState()
   &Drag(Drag=4, Use Linear Drag=true, Use Rotational Drag=false)
   &GravityForce(Gravity=<0 0 -50>)
   &SolveForcesAndVelocity(Clamp Velocity=false, Limit Acceleration=false)
   &SpriteRotationRate(Rotation Rate=&Multiply_Float(
     A=$Particles.InitSpriteRotRate
     B=&FloatFromCurve(
      FloatCurve=DataInterface() {propertyValues {Curve {keys [
          {arriveTangent=-3.9021, interpMode=RCIM_Cubic, leaveTangent=-3.9021, leaveTangentWeight=1.3778, tangentMode=RCTM_User, time=0, value=1}
          {arriveTangentWeight=0.3436, interpMode=RCIM_Cubic, tangentMode=RCTM_User, time=0.75, value=0.1}
         ]}}}
      Scale Curve=50
     )
    ))
   &ScaleSpriteSize(
    Scale Sprite Size Mode=::NewEnumerator0
    Uniform Scale Factor=&FloatFromCurve(FloatCurve=DataInterface() {propertyValues {Curve {keys [
         {interpMode=RCIM_Cubic, leaveTangent=1.0601, tangentMode=RCTM_Break, time=0, value=0.2}
         {arriveTangent=0.5926, interpMode=RCIM_Cubic, tangentMode=RCTM_Break, time=0.95, value=1}
         <at 1 0>
        ]}}})
   )
   &DynamicMaterialParameters(
    Index 0 Param 1=&Add_Float(
     A=&RandomRangeFloat(Minimum=0, Maximum=1)
     B=&FloatFromCurve(
      FloatCurve=DataInterface() {propertyValues {Curve {keys [{tangentMode=RCTM_User, time=0, value=0}, {tangentMode=RCTM_User, time=1, value=1}]}}}
      CurveIndex=&Add_Float(A=&Multiply_Float(A=$Particles.Initial.MaterialRandom, B=1), B=$Particles.NormalizedAge)
      Scale Curve=0.1
     )
    )
    Index 0 Param 2=&FloatFromCurve(FloatCurve=DataInterface() {propertyValues {Curve {keys [<at 0 1>,<at 1 0>]}}})
    Index 0 Param 3=&FloatFromCurve(
     FloatCurve=DataInterface() {propertyValues {Curve {keys [
         {interpMode=RCIM_Cubic, leaveTangent=16.5652, leaveTangentWeight=0.4385, tangentMode=RCTM_Break, time=0, value=0}
         {arriveTangentWeight=0.0333, interpMode=RCIM_Cubic, leaveTangentWeight=0.9477, tangentMode=RCTM_Break, time=0.1, value=1}
         {arriveTangent=-1.7159, interpMode=RCIM_Cubic, leaveTangent=-1.7159, tangentMode=RCTM_User, time=0.3, value=0.2}
         {arriveTangent=0.2347, interpMode=RCIM_Cubic, leaveTangent=0.2723, tangentMode=RCTM_Break, time=0.6, value=0}
        ]}}}
     Scale Curve=0
    )
    Index 0 Param 4=&FloatFromCurve(FloatCurve=DataInterface() {propertyValues {Curve {keys [{interpMode=RCIM_Cubic, tangentMode=RCTM_Break, time=0.5, value=1}, {interpMode=RCIM_Cubic, tangentMode=RCTM_Break, time=1, value=0}]}}})
    Index 1 Param 1=&FloatFromCurve(FloatCurve=DataInterface() {propertyValues {Curve {keys [
         {interpMode=RCIM_Cubic, leaveTangentWeight=0.0833, tangentMode=RCTM_Break, time=0, value=0}
         {arriveTangentWeight=0.0873, interpMode=RCIM_Cubic, tangentMode=RCTM_Break, time=0.25, value=1}
        ]}}})
    Index 1 Param 2=&FloatFromCurve(
     FloatCurve=DataInterface() {propertyValues {Curve {keys [
         {interpMode=RCIM_Cubic, tangentMode=RCTM_Break, time=0, value=0.2}
         {interpMode=RCIM_Cubic, leaveTangent=0.9066, tangentMode=RCTM_Break, time=1, value=2}
        ]}}}
     Scale Curve=&RandomRangeFloat(Minimum=0.1, Maximum=0.2)
    )
    Index 1 Param 3=&FloatFromCurve(FloatCurve=DataInterface() {propertyValues {Curve {keys [
         {interpMode=RCIM_Cubic, leaveTangent=6.3917, tangentMode=RCTM_Break, time=0, value=0}
         {arriveTangent=1.7549, interpMode=RCIM_Cubic, leaveTangent=10.6767, tangentMode=RCTM_User, time=0.25, value=1}
        ]}}})
    Index 1 Param 4=&FloatFromCurve(
     FloatCurve=DataInterface() {propertyValues {Curve {keys [{arriveTangent=-0.136, interpMode=RCIM_Cubic, leaveTangent=-0.136, tangentMode=RCTM_User, time=0, value=1}, <at 1 0.95>]}}}
     Scale Curve=2
    )
    Index 2 Param 1=&FloatFromCurve(
     FloatCurve=DataInterface() {propertyValues {Curve {keys [
         {interpMode=RCIM_Cubic, leaveTangent=4.2485, tangentMode=RCTM_Break, time=0, value=0}
         {arriveTangent=1.0548, interpMode=RCIM_Cubic, leaveTangent=1.0548, tangentMode=RCTM_User, time=0.17, value=0.5715}
         <at 1 1>
        ]}}}
     Scale Curve=3
    )
    Index 2 Param 2=&FloatFromCurve(
     FloatCurve=DataInterface() {propertyValues {Curve {keys [{interpMode=RCIM_Cubic, tangentMode=RCTM_User, time=0, value=0}, {interpMode=RCIM_Cubic, tangentMode=RCTM_User, time=1, value=1}]}}}
     Scale Curve=0.3
    )
    Index 2 Param 3=&FloatFromCurve(
     FloatCurve=DataInterface() {propertyValues {Curve {keys [
         {interpMode=RCIM_Cubic, leaveTangent=-15.7052, leaveTangentWeight=0.1411, tangentMode=RCTM_Break, time=0, value=1}
         {interpMode=RCIM_Cubic, tangentMode=RCTM_Break, time=0.1, value=0}
         {arriveTangent=0.7532, interpMode=RCIM_Cubic, leaveTangent=0.7532, tangentMode=RCTM_User, time=0.7, value=0.5}
         {arriveTangent=2.358, arriveTangentWeight=1.057, interpMode=RCIM_Cubic, tangentMode=RCTM_Break, time=1, value=1}
        ]}}}
     Scale Curve=5
    )
    Index 2 Param 4=&RandomRangeFloat(Minimum=1.5, Maximum=1.75)
    Module.Param0WriteEnabled=true
    Module.Param0WriteEnabled1=true
    Module.Param0WriteEnabled2=true
    Module.Param1WriteEnabled=false
    Module.Param1WriteEnabled1=true
    Module.Param1WriteEnabled2=true
    Module.Param2WriteEnabled=true
    Module.Param2WriteEnabled1=true
    Module.Param2WriteEnabled2=true
    Module.Param3WriteEnabled=true
    Module.Param3WriteEnabled1=true
    Module.Param3WriteEnabled2=true
    Write Parameter Index 1=true
    Write Parameter Index 2=true
   )
   /SetVariables(Particles.NewPosition:NiagaraPosition=&FN_FrontEndVector3Modifier(
     Gameplay=Linked() {linkedVariable {name=Particles.Position, type=NiagaraPosition}}
     Front End=&AddVectorToPosition(
      Position=$Particles.Position
      Vector=&Multiply_VectorByFloat(
       Vector=$Engine.Owner.SystemYAxis
       Float=&FloatFromCurve(
        FloatCurve=DataInterface() {propertyValues {Curve {defaultValue=0, keys [{tangentMode=RCTM_User, time=0, value=0}, {tangentMode=RCTM_User, time=1, value=1}]}}}
        Scale Curve=-50
       )
      )
     )
    )) {isSetParameters=true}
   &ScaleColor(
    Scale Alpha=&FloatFromCurve(
     FloatCurve=DataInterface() {propertyValues {Curve {keys [
         {arriveTangent=6.442, interpMode=RCIM_Cubic, leaveTangent=6.442, tangentMode=RCTM_User, time=0, value=0}
         {interpMode=RCIM_Cubic, tangentMode=RCTM_User, time=0.1, value=1}
         <at 1 0>
        ]}}}
     Scale Curve=&RandomRangeFloat(Minimum=0.75, Maximum=1)
    )
    ScaleA=true
    ScaleRGB=false
   )
  ]
  rendererClasses [&NiagaraSpriteRendererProperties]
  renderers [
   {rendererClass=&NiagaraSpriteRendererProperties}
   {rendererClass=&NiagaraSpriteRendererProperties, rendererIndex=1}
  ]
 }
 Ground_Cracks001 {
  emitterUpdateScript [
   &EmitterState()
   &SpawnBurst_Instantaneous(Use Spawn Probability=false, UseLoopCountLimit=false)
   &SpawnBurst_Instantaneous(Spawn Group=1, Use Spawn Probability=false, UseLoopCountLimit=false) {module=SpawnBurst_Instantaneous001}
  ]
  particleSpawnScript [
   &InitializeParticle(
    Lifetime=&MakeCustomFloatFromBool(Bool=$User.bIsFrontEndPreview, True Float=1, False Float=1.5)
    Position Offset=&MaskVector(Mask Source=1, Mask Parameter=$Emitter.SpawnGroup, Mask=0, Masked Vector=<0 0 0>, Passthrough Vector=<-500 0 0>)
    Sprite Rotation Mode=::NewEnumerator2
    Sprite Size Mode=::NewEnumerator1
    Uniform Sprite Size Max=80
    Uniform Sprite Size Min=60
    UsePositionOffset=true
    Write Material Random=true
   ) {module=InitializeParticle001}
   /SetVariables(Particles.RandomDecalYZ:NiagaraFloat=&RandomRangeFloat(Minimum=130, Maximum=150)) {isSetParameters=true}
   /SetVariables(Particles.DecalSize:Vector3f=&MakeVector(X=75, Y=$Particles.RandomDecalYZ, Z=$Particles.RandomDecalYZ)) {isSetParameters=true}
   &InitialMeshOrientation(
    Mesh Orientation Mode=::NewEnumerator2
    Orientation Vector=<0 0 1>
    Rotation=&RandomRangeVector(Minimum=<-1 0 0>, Maximum=<1 0 0>)
    Use Rotation Vector=true
   )
   /SetVariables(Particles.VisibilityTag=/Niagara/DynamicInputs/Helpers/MaskInt(
     Mask Source:NiagaraInt32=1
     Mask Parameter:NiagaraInt32=$Emitter.SpawnGroup
     Mask:NiagaraInt32=0
     Masked Int:NiagaraInt32=0
     Passthrough Int:NiagaraInt32=1
    )) {isSetParameters=true}
   &SpriteFacingAndAlignment(Sprite Alignment=<1 0 0>, Sprite Facing=<0 0 1>, WriteAlignment=true, WriteFacing=true) {module=SpriteFacingAndAlignment001}
   /Niagara/Modules/Spawn/Decal/InitialDecalOrientation(
    Orientation Vector:Vector3f=<0 0 1>
    Rotation:Vector3f=&RandomRangeVector(Minimum=<0 0 0>, Maximum=<0 0 1>)
    Use Rotation Vector=true
   )
   /SetVariables(
    Particles.GlowColor=&Multiply_LinearColorByFloat(Linear Color=<rgba 2 0.4288 0.0373 1>, Float=1)
    Particles.GlowDyn:Vector4f=<1 1 0 0>
    Particles.GlowPos:NiagaraPosition=&AddVectorToPosition(Position=$Engine.Emitter.SimulationPosition, Vector=<2 0 0>)
   ) {isSetParameters=true}
   &CameraOffset(Camera Offset Amount=&MaskFloat(Mask Source=1, Mask Parameter=$Emitter.SpawnGroup, Mask=0, Masked Float=0, Passthrough Float=20))
  ]
  particleUpdateScript [
   &ParticleState()
   &ScaleSpriteSize(Uniform Curve Sprite Scale=DataInterface() {propertyValues {Curve {keys [
        {arriveTangent=230.8238, interpMode=RCIM_Cubic, leaveTangent=230.8238, tangentMode=RCTM_User, time=0, value=0}
        {arriveTangent=0.8135, interpMode=RCIM_Cubic, leaveTangent=0.8135, tangentMode=RCTM_User, time=0.01, value=1}
       ]}}})
   &ScaleColor(
    Scale Alpha=&FloatFromCurve(FloatCurve=DataInterface() {propertyValues {Curve {keys [
         {interpMode=RCIM_Cubic, tangentMode=RCTM_User, time=0, value=0}
         <at 0.025 1>
         {arriveTangent=-1.0256, interpMode=RCIM_Cubic, leaveTangent=-1.0256, time=0.7, value=1}
         <at 1 0 RCIM_Cubic>
        ]}}})
    ScaleA=true
    ScaleRGB=true
   )
   &ScaleSpriteSize(Uniform Curve Sprite Scale=DataInterface() {propertyValues {Curve {keys [{arriveTangent=139.099, interpMode=RCIM_Cubic, leaveTangent=139.099, tangentMode=RCTM_User, time=0, value=0}, <at 0.015 1 RCIM_Cubic>]}}}) {module=ScaleSpriteSize001}
   /SetVariables(Particles.DecalFade:NiagaraFloat=&FloatFromCurve(
     FloatCurve=DataInterface() {propertyValues {Curve {keys [
         {interpMode=RCIM_Cubic, tangentMode=RCTM_User, time=0, value=0}
         <at 0.025 1>
         {arriveTangent=-1.0256, interpMode=RCIM_Cubic, leaveTangent=-1.0256, time=0.7, value=1}
         <at 1 0 RCIM_Cubic>
        ]}}}
     Scale Curve=&MakeCustomFloatFromBool(Bool=$User.bIsFrontEndPreview, True Float=0.5, False Float=1)
    )) {isSetParameters=true}
   /SetVariables(Particles.DecalSize:Vector3f=&Multiply_VectorByFloat(
     Vector=$Particles.Initial.DecalSize
     Float=&FloatFromCurve(FloatCurve=DataInterface() {propertyValues {Curve {keys [{arriveTangent=139.099, interpMode=RCIM_Cubic, leaveTangent=139.099, tangentMode=RCTM_User, time=0, value=0}, <at 0.015 1 RCIM_Cubic>]}}})
    )) {isSetParameters=true}
   &SampleSkeletalMesh(
    AdditionalYawPitchRoll=false
    Mesh Sampling Type:ENiagaraSkelSamplingModeFull=::NewEnumerator5
    Skeletal Mesh=DataInterface() {propertyValues {
      DefaultMesh=null
      ExcludeBoneName=root
      OptionalComponentClass=null
      PreviewMesh=/Game/Characters/Player/Male/Medium/Base/SK_M_MALE_Base_Skeleton
      SkinningMode=SkinOnTheFly
      SoftSourceActor=null
      SourceMode=Default
      UvSetIndex=0
      WholeMeshLOD=-1
      bExcludeBone=false
      bReadDeformedGeometry=true
      bRequireCurrentFrameData=true
      ComponentTags []
      FilteredBones []
      FilteredSockets [Root_Effect]
      MeshUserParameter {parameter {name=None, typeDefHandle {registeredTypeIndex=1}, varData []}}
      SamplingRegions []
     }}
    Socket Sampling Mode:ENiagaraSocketSamplingMode=::NewEnumerator5
   ) {module=SampleSkeletalMesh001}
   /SetVariables(Particles.Position=&MakeCustomVectorFromBool(
     Bool=$User.bIsFrontEndPreview
     True Vector=&AddVectorToPosition(
      Position=$Particles.Position
      Vector=&Multiply_VectorByFloat(
       Vector=&RotateVectorByQuaternion(Vector To Rotate=<1 0 0>, Quaternion=$StackContext.SampleSkeletalMesh001.SampledRotation)
       Float=-2.5
      )
     )
     False Vector=Linked() {linkedVariable {name=Particles.Position, type=NiagaraPosition}}
    )) {enabled=false, isSetParameters=true}
   &SolveForcesAndVelocity(Clamp Velocity=false, Limit Acceleration=false)
   &Light_Attributes(Enabled=1, Light Radius=1, Radius=1)
   /SetVariables(Particles.LightPos:NiagaraPosition=&AddVectorToPosition(Position=$Particles.Position, Vector=<0 0 5>)) {isSetParameters=true}
   /SetVariables(
    Particles.GlowColor=&MultiplyLinearColor(
     Linear Color=$Particles.Initial.GlowColor
     Scale Factor=&MakeVector4(
      X=3
      Y=3
      Z=3
      W=&FloatFromCurve(FloatCurve=DataInterface() {propertyValues {Curve {
          defaultValue=1
          keys [
           {interpMode=RCIM_Cubic, tangentMode=RCTM_User, time=0, value=0}
           {interpMode=RCIM_Cubic, tangentMode=RCTM_User, time=0.0327, value=1}
           {arriveTangent=-1.0338, interpMode=RCIM_Cubic, leaveTangent=-1.0338, time=0.4147, value=0.7683}
           {tangentMode=RCTM_User, time=1, value=0}
          ]
         }}})
     )
    )
    Particles.GlowSize=<40 40>
   ) {isSetParameters=true}
  ]
  rendererClasses [&NiagaraDecalRendererProperties, &NiagaraSpriteRendererProperties]
  renderers [
   {rendererClass=&NiagaraDecalRendererProperties, rendererIndex=0}
   {rendererClass=&NiagaraSpriteRendererProperties, rendererIndex=1}
  ]
 }
 HT_Debris {
  simTarget=GPUComputeSim
  emitterUpdateScript [
   &EmitterState()
   &SpawnBurst_Instantaneous(Spawn Count=25, Spawn Group=1, Use Spawn Probability=false, UseLoopCountLimit=false) {module=SpawnBurst_Instantaneous001}
   &SpawnBurst_Instantaneous(Spawn Count=45, Use Spawn Probability=false, UseLoopCountLimit=false)
  ]
  particleSpawnScript [
   &InitializeParticle(
    Color=&RandomRangeLinearColor(Minimum=<rgba 1 0.5174 0.2029 1>, Maximum=<rgba 1 0.0586 0 1>)
    Lifetime=&UniformRangedFloat(
     Minimum=&MaskFloat(Mask Source=1, Mask Parameter=&SpawnGroup(), Mask=0, Masked Float=0.5, Passthrough Float=0.8)
     Maximum=&MaskFloat(Mask Source=1, Mask Parameter=&SpawnGroup(), Mask=0, Masked Float=2, Passthrough Float=1.2)
    )
    Mesh Renderer Array Visibility Mode=::NewEnumerator3
    Mesh Renderer Info=DataInterface() {propertyValues {MeshRenderer=&self:HT_Debris_4.NiagaraMeshRendererProperties_0}}
    Mesh Scale Max=<0.5 0.5 0.5>
    Mesh Scale Min=<0.2 0.2 0.5>
    Mesh Scale Mode=::NewEnumerator2
    Sprite Rotation Mode=::NewEnumerator2
    Sprite Size Max=<3 3>
    Sprite Size Min=<1.5 1.5>
    Sprite Size Mode=::NewEnumerator2
    UsePositionOffset=false
    Write Material Random=true
   ) {module=InitializeParticle001}
   /SetVariables(Particles.RockSize:NiagaraFloat=&Multiply_Float(
     A=&RandomRangeFloat(Minimum=0.4, Maximum=1.4)
     B=&MaskFloat(Mask Source=1, Mask Parameter=&SpawnGroup(), Mask=0, Masked Float=1, Passthrough Float=2)
    )) {isSetParameters=true}
   /SetVariables(Particles.SpriteSize=&Multiply_Vector2D_ByFloat(Float=$Particles.RockSize, Vector2D=$Particles.SpriteSize)) {isSetParameters=true}
   /SetVariables(Particles.Scale=&Multiply_VectorByFloat(Vector=$Particles.Scale, Float=$Particles.RockSize)) {isSetParameters=true}
   &ShapeLocation(
    MaskSpawnGroup=false
    Offset Mode=::NewEnumerator0
    Override Seed=false
    Ring Radius=&MaskFloat(Mask Source=1, Mask Parameter=&SpawnGroup(), Mask=0, Masked Float=20, Passthrough Float=10)
    Shape Primitive=::NewEnumerator4
    UseOwnerScale=true
   ) {enabled=false}
   &ShapeLocation(
    Cone Surface Distribution=0.5119
    MaskSpawnGroup=false
    Non Uniform Scale=<1 1 -1>
    Offset=<0 0 50>
    Offset Mode=::NewEnumerator0
    Override Seed=false
    Shape Primitive=::NewEnumerator5
    UseOwnerScale=true
   ) {module=ShapeLocation001}
   &AddVelocity(
    Cone Angle=25
    Cone Axis=<0 0 1>
    Module.UseConeFalloff=false
    Module.UseVelDistribution=false
    Override Seed=false
    Velocity Mode=::NewEnumerator2
    Velocity Speed=&Multiply_Float(
     A=&RandomRangeFloat(Minimum=250, Maximum=500)
     B=&MaskFloat(Mask Source=1, Mask Parameter=&SpawnGroup(), Mask=0, Masked Float=1, Passthrough Float=0)
    )
   )
   &AddRotationalVelocity(Rotation Rate=&RandomVector(Vector Scale=&MaskFloat(Mask Source=1, Mask Parameter=&SpawnGroup(), Mask=0, Masked Float=1, Passthrough Float=0)))
   /SetVariables(Particles.Velocity=&Multiply_VectorByFloat(
     Vector=$Particles.Velocity
     Float=&FloatFromCurve(FloatCurve=DataInterface() {propertyValues {Curve {keys [<at 0.4 2>,<at 1.4 1>]}}}, CurveIndex=$Particles.RockSize)
    )) {isSetParameters=true}
   /Niagara/Modules/Update/SubUV/SubUVAnimation(Random SubUV Lookup Index:NiagaraFloat=$Particles.RandomFloat, SubUV Animation Mode=::NewEnumerator1) {module=SubUVAnimation001}
   /SetVariables(Particles.VisibilityTag=&MakeCustomIntFromBool(
     Bool=&Boolean_OrOperation(
      Bool1=&IsPlatformSetActive(New Platform Set=DataInterface() {propertyValues {Platforms {qualityLevelMask=-2}}})
      Bool2=$User.bIsFrontEndPreview
     )
     True INT=&RandomRangeInt(
      Minimum=&MakeIntFromFloat(Float=&MaskFloat(Mask Source=1, Mask Parameter=&SpawnGroup(), Mask=0, Masked Float=0, Passthrough Float=1))
      Maximum=1
     )
     False INT=0
    )) {isSetParameters=true}
   &InitialMeshOrientation(Use Rotation Vector=false)
  ]
  particleUpdateScript [
   &ParticleState() {module=UpdateAge}
   &Drag(Use Linear Drag=true, Use Rotational Drag=false)
   &GravityForce(Gravity=&Multiply_VectorByFloat(
     Vector=<0 0 -980>
     Float=&MaskFloat(Mask Source=1, Mask Parameter=&SpawnGroup(), Mask=0, Masked Float=1, Passthrough Float=0)
    ))
   &Collision(
    Advanced Aging Rate=1
    Age Colliding Particles=false
    Collision Enabled=&Boolean_OrOperation(
     Bool1=&IsPlatformSetActive(New Platform Set=DataInterface() {propertyValues {Platforms {qualityLevelMask=-4}}})
     Bool2=$User.bIsFrontEndPreview
    )
    Control Roll On Collision=true
    Enable Rest State=false
    GPU Collision Type=::NewEnumerator2
    Module.EnableMaxCollisionCount=false
    Module.Randomize Collision Normal Vector=true
    Particle Radius Scale=0.5
    Restitution=0.4
    ScaleRateOfRotation=false
   )
   &SolveForcesAndVelocity(Clamp Velocity=false, Limit Acceleration=false)
   &SpriteRotationRate(Rotation Rate=&Multiply_Float(
     A=$Particles.RandomFloat
     B=&FN_PickFloatDependOnPlatform(TargetPlatform=DataInterface() {propertyValues {Platforms {qualityLevelMask=-13}}}, TargetPlatformActive=0, TargetPlatformNoActive=4)
    ))
   &ScaleSpriteSize(
    Uniform Curve Scale=2
    Uniform Curve Sprite Scale=DataInterface() {propertyValues {Curve {keys [
        <at 0 1>
        {arriveTangent=-1, interpMode=RCIM_Cubic, leaveTangent=-1, time=0.8, value=1}
        {arriveTangent=-5.7307, interpMode=RCIM_Cubic, leaveTangent=-5.7307, tangentMode=RCTM_User, time=1, value=0}
       ]}}}
   )
   &ScaleMeshSize(Scale Factor=&VectorFromFloat(Value=&FloatFromCurve(FloatCurve=DataInterface() {propertyValues {Curve {keys [
          <at 0 1>
          {arriveTangent=-1, interpMode=RCIM_Cubic, leaveTangent=-1, time=0.8, value=1}
          {arriveTangent=-5.7307, interpMode=RCIM_Cubic, leaveTangent=-5.7307, tangentMode=RCTM_User, time=1, value=0}
         ]}}})))
   &SpriteRotationRate(Rotation Rate=&RandomRangeFloat(Minimum=-300, Maximum=300)) {module=SpriteRotationRate001}
   &SampleSkeletalMesh(
    AdditionalYawPitchRoll=false
    Mesh Sampling Type:ENiagaraSkelSamplingModeFull=::NewEnumerator5
    Sampled Orientation Coordinate Space:ENiagaraCoordinateSpace=::World
    Skeletal Mesh=DataInterface() {propertyValues {
      DefaultMesh=null
      ExcludeBoneName=root
      OptionalComponentClass=null
      PreviewMesh=&SK_M_MALE_Base_Skeleton
      SkinningMode=SkinOnTheFly
      SoftSourceActor=null
      SourceMode=Default
      UvSetIndex=0
      WholeMeshLOD=-1
      bExcludeBone=false
      bReadDeformedGeometry=true
      bRequireCurrentFrameData=true
      ComponentTags []
      FilteredBones []
      FilteredSockets [Root_Effect]
      MeshUserParameter {parameter {name=None, typeDefHandle {registeredTypeIndex=1}, varData []}}
      SamplingRegions []
     }}
    Socket Sampling Mode:ENiagaraSocketSamplingMode=::NewEnumerator5
   ) {module=SampleSkeletalMesh001}
   /SetVariables(Particles.Position=&MakeCustomVectorFromBool(
     Bool=$User.bIsFrontEndPreview
     True Vector=&AddVectorToPosition(
      Position=$Particles.Position
      Vector=&Multiply_VectorByFloat(
       Vector=&RotateVectorByQuaternion(Vector To Rotate=<1 0 0>, Quaternion=$StackContext.SampleSkeletalMesh001.SampledRotation)
       Float=-5
      )
     )
     False Vector=Linked() {linkedVariable {name=Particles.Position, type=NiagaraPosition}}
    )) {enabled=false, isSetParameters=true}
   &ScaleColor(
    Scale RGB=&VectorFromFloat(Value=&FloatFromCurve(
      FloatCurve=DataInterface() {propertyValues {Curve {keys [<at 0 1>,<at 0.2 0>]}}}
      Scale Curve=&RandomRangeFloat(Minimum=2, Maximum=4)
     ))
    ScaleA=true
    ScaleRGB=true
   )
  ]
  rendererClasses [&NiagaraSpriteRendererProperties, &NiagaraMeshRendererProperties]
  renderers [
   {rendererClass=&NiagaraSpriteRendererProperties}
   {rendererClass=&NiagaraMeshRendererProperties, rendererIndex=1}
  ]
 }
 LavaSplash001 {
  emitterUpdateScript [
   &EmitterState()
   &SpawnBurst_Instantaneous(Use Spawn Probability=false, UseLoopCountLimit=false)
  ]
  particleSpawnScript [
   &InitializeParticle(
    Lifetime=0.433
    Sprite Rotation Angle=&RandomRangeFloat(Minimum=0.033, Maximum=0.466)
    Sprite Rotation Mode=::NewEnumerator1
    Sprite Size Mode=::NewEnumerator3
    Uniform Sprite Size=166
    UsePositionOffset=false
    Write Material Random=true
   )
   &AddVelocity(Velocity=<0 0 66>)
   &CameraOffset(Camera Offset Amount=11)
   /SetVariables(
    Particles.Sprite_01_CameraOffset=35
    Particles.Sprite_01_Rotation:NiagaraFloat=&RandomRangeFloat(Minimum=-45, Maximum=45)
    Particles.Sprite_01_Size:Vector2f=&Vector2DFromFloat(Value=50)
   ) {isSetParameters=true}
  ]
  particleUpdateScript [
   &ParticleState()
   &ScaleSpriteSize(Uniform Curve Sprite Scale=DataInterface() {propertyValues {Curve {keys [
        {arriveTangent=1.0404, interpMode=RCIM_Cubic, leaveTangent=1.0404, tangentMode=RCTM_Break, time=0, value=0.25}
        {arriveTangent=-0.0405, interpMode=RCIM_Cubic, leaveTangent=-0.0405, tangentMode=RCTM_User, time=1, value=1}
       ]}}})
   &ScaleColor(
    Scale Alpha=&FloatFromCurve(FloatCurve=DataInterface() {propertyValues {Curve {keys [
         {interpMode=RCIM_Cubic, leaveTangent=-0.8793, tangentMode=RCTM_Break, time=0, value=1}
         {arriveTangent=0.0015, interpMode=RCIM_Cubic, leaveTangent=-1.7296, tangentMode=RCTM_Break, time=1, value=0}
        ]}}})
    ScaleA=true
    ScaleRGB=true
   )
   &GravityForce()
   &Drag(Use Linear Drag=true, Use Rotational Drag=true)
   &SolveForcesAndVelocity(Clamp Velocity=false, Limit Acceleration=false)
   &DynamicMaterialParameters(
    Index 0 Param 1=0.1
    Module.Index0Param0Write=true
    Module.Index0Param1Write=true
    Module.Index0Param2Write=true
    Module.Index0Param3Write=true
   )
   /SetVariables(
    Particles.Sprite_01_Color:LinearColor=&ColorFromCurve(
     ColorCurve=DataInterface() {propertyValues {
       AlphaCurve {keys [
         {interpMode=RCIM_Cubic, tangentMode=RCTM_User, time=0.2, value=0.9}
         {arriveTangent=-2.2354, arriveTangentWeight=0.267, interpMode=RCIM_Cubic, leaveTangent=-2.2354, tangentMode=RCTM_User, time=1, value=0.002}
        ]}
       BlueCurve {keys [
         <at 0.2022 0>
         {arriveTangent=0.0367, interpMode=RCIM_Cubic, leaveTangent=0.0367, time=0.4797, value=0}
         {arriveTangent=0.0367, interpMode=RCIM_Cubic, leaveTangent=0.0367, time=0.6103, value=0}
         <at 0.6586 0.0168>
        ]}
       GreenCurve {keys [
         <at 0.2022 0.9302>
         {arriveTangent=-2.0302, interpMode=RCIM_Cubic, leaveTangent=-2.0302, time=0.4797, value=0.25}
         {arriveTangent=-2.0302, interpMode=RCIM_Cubic, leaveTangent=-2.0302, time=0.6103, value=0.0833}
         <at 0.6586 0.0037>
        ]}
       RedCurve {keys [
         <at 0.2022 3>
         {arriveTangent=-6.3336, interpMode=RCIM_Cubic, leaveTangent=-6.3336, time=0.4797, value=3}
         {arriveTangent=-6.3336, interpMode=RCIM_Cubic, leaveTangent=-6.3336, time=0.6103, value=1}
         <at 0.6586 0.1094>
        ]}
      }}
     CurveIndex=&Divide_Float(A=$Particles.NormalizedAge, B=0.7)
    )
    Particles.Sprite_01_Size:Vector2f=&Multiply_Vector2D(
     A=$Particles.Initial.Sprite_01_Size
     B=&Vector2DFromFloat(Value=&FloatFromCurve(
       FloatCurve=DataInterface() {propertyValues {Curve {keys [
           {arriveTangent=16.2435, interpMode=RCIM_Cubic, leaveTangent=16.2435, tangentMode=RCTM_User, time=0, value=0}
           {arriveTangent=4.4995, interpMode=RCIM_Cubic, leaveTangent=4.4995, tangentMode=RCTM_User, time=0.0956, value=0.7347}
           {arriveTangent=-0.7763, interpMode=RCIM_Cubic, leaveTangent=-0.7763, tangentMode=RCTM_User, time=1, value=2}
          ]}}}
       CurveIndex=&Divide_Float(A=$Particles.NormalizedAge, B=0.7)
      ))
    )
   ) {isSetParameters=true}
  ]
  rendererClasses [&NiagaraSpriteRendererProperties]
  renderers [
   {rendererClass=&NiagaraSpriteRendererProperties}
   {rendererClass=&NiagaraSpriteRendererProperties, rendererIndex=1}
  ]
 }
 RingBurst {
  emitterUpdateScript [
   &EmitterState()
   &SpawnRate(
    SpawnRate=&FloatFromCurve(FloatCurve=DataInterface() {propertyValues {Curve {keys [<at 0.2 1>,<at 0.85 0>]}}}, CurveIndex=$System.Age, Scale Curve=35)
    Use Spawn Probability=false
   )
  ]
  particleSpawnScript [
   /SetVariables(Particles.Random:NiagaraInt32=&RandomRangeInt(Minimum=0, Maximum=1)) {isSetParameters=true}
   &InitializeParticle(
    Color Maximum=<rgba 1 0.6294 0.6199 1>
    Color Minimum=<rgba 1 0.9684 0.9781 1>
    Color Mode=::NewEnumerator2
    Lifetime=&RandomRangeFloat(Minimum=0.3, Maximum=0.8)
    Sprite Rotation Mode=::NewEnumerator2
    Sprite Size Mode=::NewEnumerator3
    Uniform Sprite Size=&Multiply_Float(
     A=$System.FX_Scale
     B=&MaskFloat(
      Mask Source=1
      Mask Parameter=$Particles.Random
      Mask=0
      Masked Float=&RandomRangeFloat(Minimum=75, Maximum=105)
      Passthrough Float=&RandomRangeFloat(Minimum=150, Maximum=210)
     )
    )
    UsePositionOffset=false
    Write Material Random=true
   )
   &ShapeLocation(
    MaskSpawnGroup=false
    Override Seed=false
    Ring Radius=&Multiply_Float(A=4, B=$System.FX_Scale)
    Rotation Mode=::NewEnumerator1
    Shape Primitive=::NewEnumerator4
    UseOwnerScale=true
   )
   &AddVelocity(
    Cone Angle=30
    Cone Axis=<0 0 1>
    Module.UseConeFalloff=false
    Module.UseVelDistribution=false
    Override Seed=false
    Velocity Mode=::NewEnumerator2
    Velocity Speed=&Multiply_Float(
     A=$System.FX_Scale
     B=&MaskFloat(Mask Source=1, Mask Parameter=$Particles.Random, Mask=0, Masked Float=200, Passthrough Float=200)
    )
   )
   &AddVelocity(
    Velocity=&MaskVector(Mask Source=1, Mask Parameter=$Particles.Random, Mask=0, Masked Vector=<0 0 0>, Passthrough Vector=<0 0 25>)
    Velocity Speed Scale=&RandomRangeFloat(Minimum=0.9, Maximum=1)
   ) {module=AddVelocity001}
   &CameraOffset(Camera Offset Amount=&RandomRangeFloat(Minimum=15, Maximum=30))
  ]
  particleUpdateScript [
   &Drag(Use Linear Drag=true, Use Rotational Drag=true)
   &self:ScratchModule_02() {module=ScratchModule_02}
   &SolveForcesAndVelocity(Clamp Velocity=false, Limit Acceleration=false)
   &ParticleState()
   &SpriteRotationRate(Rotation Rate=&FloatFromCurve(
     FloatCurve=DataInterface() {propertyValues {Curve {keys [<at 0 1>,<at 1 -1>]}}}
     CurveIndex=$Output.ScratchModule_02.SpriteRotationDirectionMask
     Scale Curve=-1000
    ))
   &ScaleSpriteSize(
    Scale Sprite Size Mode=::NewEnumerator0
    Uniform Scale Factor=&MaskFloat(
     Mask Source=1
     Mask Parameter=$Particles.Random
     Mask=0
     Masked Float=&FloatFromCurve(FloatCurve=DataInterface() {propertyValues {Curve {keys [
          {arriveTangent=0.2827, interpMode=RCIM_Cubic, leaveTangent=0.2827, tangentMode=RCTM_User, time=0, value=0.6}
          {interpMode=RCIM_Cubic, tangentMode=RCTM_User, time=1, value=1}
         ]}}})
     Passthrough Float=&FloatFromCurve(FloatCurve=DataInterface() {propertyValues {Curve {keys [
          {arriveTangent=-0.1234, interpMode=RCIM_Cubic, leaveTangent=-0.1234, tangentMode=RCTM_User, time=0, value=0.3}
          {interpMode=RCIM_Cubic, tangentMode=RCTM_User, time=0.7, value=0.8}
         ]}}})
    )
   )
   &ScaleColor(
    Scale Alpha=&FloatFromCurve(
     FloatCurve=DataInterface() {propertyValues {Curve {keys [
         {interpMode=RCIM_Cubic, tangentMode=RCTM_User, time=0, value=0}
         {interpMode=RCIM_Cubic, tangentMode=RCTM_User, time=0.3, value=1}
         {interpMode=RCIM_Cubic, tangentMode=RCTM_User, time=1, value=0}
        ]}}}
     Scale Curve=&Multiply_Float(A=0.1, B=1)
    )
    Scale RGB=&MakeVectorFromLinearColorRGB(Color=&MaskLinearColor(
      Mask Source=1
      Mask Parameter=&RandomRangeInt(Minimum=0, Maximum=1)
      Mask=0
      Masked Linear Color=&ColorFromCurve(ColorCurve=DataInterface() {propertyValues {
         AlphaCurve {keys [<at 0 0.85 RCIM_Cubic>]}
         BlueCurve {keys [<at 0 0 RCIM_Cubic>,<at 0.4931 0>,<at 1 0 RCIM_Cubic>]}
         GreenCurve {keys [<at 0 1.6122 RCIM_Cubic>,<at 0.4931 0.6752>,<at 1 0 RCIM_Cubic>]}
         RedCurve {keys [<at 0 10 RCIM_Cubic>,<at 0.4931 2>,<at 1 1 RCIM_Cubic>]}
        }})
      Passthrough Linear Color=&ColorFromCurve(ColorCurve=DataInterface() {propertyValues {
         AlphaCurve {keys [<at 0 0.85 RCIM_Cubic>]}
         BlueCurve {keys [<at 0 0.24 RCIM_Cubic>,<at 0.4931 0>,<at 1 0 RCIM_Cubic>]}
         GreenCurve {keys [<at 0 0.8372 RCIM_Cubic>,<at 0.4931 0.6752>,<at 1 0 RCIM_Cubic>]}
         RedCurve {keys [<at 0 10 RCIM_Cubic>,<at 0.4931 2>,<at 1 1 RCIM_Cubic>]}
        }})
     ))
    ScaleA=true
    ScaleRGB=true
   ) {module=ScaleColor001}
   /Niagara/Modules/Update/SubUV/V2/SubUVAnimation(
    "Sprite Renderer "=DataInterface() {propertyValues {SpriteRenderer=&self:NiagaraEmitter_0.NiagaraSpriteRendererProperties_0}}
    Start Frame Offset=2
    UseEndFrame=false
    UseStartFrame=false
   )
   &DynamicMaterialParameters(
    Index 0 Param 2=&MaskFloat(
     Mask Source=1
     Mask Parameter=&RandomRangeInt(Minimum=0, Maximum=1)
     Mask=0
     Masked Float=0
     Passthrough Float=&FloatFromCurve(
      FloatCurve=DataInterface() {propertyValues {Curve {keys [
          {arriveTangent=-2.5672, interpMode=RCIM_Cubic, leaveTangent=-2.5672, tangentMode=RCTM_User, time=0, value=1}
          {interpMode=RCIM_Cubic, tangentMode=RCTM_User, time=1, value=0}
         ]}}}
      Scale Curve=&RandomRangeFloat(Minimum=0.1, Maximum=0.2)
     )
    )
    Module.Index0Float0Write=true
    Module.Index0Float1Write=true
    Module.Index0Float2Write=true
    Module.Index0Float3Write=true
   )
  ]
  rendererClasses [&NiagaraSpriteRendererProperties]
  renderers [{rendererClass=&NiagaraSpriteRendererProperties}]
 }
 Shockwave {
  emitterUpdateScript [
   &EmitterState(Life Cycle Mode=::NewEnumerator1, Loop Behavior=::NewEnumerator1, UseLoopDelay=false)
   &SpawnBurst_Instantaneous(Use Spawn Probability=false, UseLoopCountLimit=false)
  ]
  particleSpawnScript [
   &InitializeParticle(
    Color=&FNTimeOfDayColor(Night Color=<rgba 0.2039 0.2396 1 0.8>, Morning Color=<rgba 1 1 1 0.8>, Day Color=<rgba 1 1 1 0.8>, Evening Color=<rgba 1 1 1 0.8>)
    Lifetime=0.4
    Mesh Scale Mode=::NewEnumerator3
    Mesh Uniform Scale=0.2
    Position Offset=<0 0 20>
    Position Offset Coordinate Space=::Local
    UsePositionOffset=true
    Write Material Random=true
   )
   &InitialMeshOrientation(Mesh Orientation Mode=::NewEnumerator2, Use Rotation Vector=false)
  ]
  particleUpdateScript [
   &ParticleState()
   &UpdateMeshOrientation(Rotation Vector=<0 0 1>)
   &DynamicMaterialParameters(
    Index 0 Param 1=&FloatFromCurve(FloatCurve=DataInterface() {propertyValues {Curve {keys [{interpMode=RCIM_Cubic, tangentMode=RCTM_User, time=0.5, value=0}, {interpMode=RCIM_Cubic, tangentMode=RCTM_User, time=1, value=1}]}}})
    Index 0 Param 3=1
    Module.Param0WriteEnabled=true
    Module.Param1WriteEnabled=true
    Module.Param2WriteEnabled=true
    Module.Param3WriteEnabled=true
   )
   &ScaleMeshSize(Scale Factor=&VectorFromCurve(
     VectorCurve=DataInterface() {propertyValues {
       XCurve {keys [
         {arriveTangent=4.1785, interpMode=RCIM_Cubic, leaveTangent=4.1785, tangentMode=RCTM_User, time=0, value=0}
         {arriveTangent=1.1179, interpMode=RCIM_Cubic, leaveTangent=1.1179, tangentMode=RCTM_User, time=0.2042, value=0.6228}
         {arriveTangent=0.0224, interpMode=RCIM_Cubic, leaveTangent=0.0224, tangentMode=RCTM_User, time=1, value=1}
        ]}
       YCurve {keys [
         {arriveTangent=4.1785, interpMode=RCIM_Cubic, leaveTangent=4.1785, tangentMode=RCTM_User, time=0, value=0}
         {arriveTangent=1.1179, interpMode=RCIM_Cubic, leaveTangent=1.1179, tangentMode=RCTM_User, time=0.2042, value=0.6228}
         {arriveTangent=0.0224, interpMode=RCIM_Cubic, leaveTangent=0.0224, tangentMode=RCTM_User, time=1, value=1}
        ]}
       ZCurve {keys [{tangentMode=RCTM_User, time=0, value=1}, {tangentMode=RCTM_User, time=1, value=1}]}
      }}
     Scale Curve=<3 3 2>
    ))
   &ScaleColor(
    Linear Color Curve=DataInterface() {propertyValues {
      AlphaCurve {keys [<at 0 0>,<at 0.2021 0.8>,<at 1 0>]}
      BlueCurve {keys [<at 0 0.1682>,<at 0.1795 0.2536>,<at 0.6246 0.3697>,<at 1 0.1357>]}
      GreenCurve {keys [<at 0 0.192>,<at 0.1795 2.5511>,<at 0.6246 0.5905>,<at 1 0.1211>]}
      RedCurve {keys [<at 0 0.3229>,<at 0.1795 20>,<at 0.6246 0.9219>,<at 1 0.3021>]}
     }}
    Scale Mode=::NewEnumerator2
   ) {module=ScaleColor001}
  ]
  rendererClasses [&NiagaraMeshRendererProperties]
  renderers [{rendererClass=&NiagaraMeshRendererProperties, rendererIndex=0}]
 }
 Voxel_On_Med001 {
  emitterUpdateScript [
   &EmitterState(Life Cycle Mode=::NewEnumerator1, Loop Behavior=::NewEnumerator1, Loop Duration=0.55, UseLoopDelay=false)
   &SpawnRate(SpawnRate=15, Use Spawn Probability=false)
  ]
  particleSpawnScript [
   &InitializeParticle(
    Color=<rgba 4 3.2896 2.4052 0.75>
    Lifetime=&UniformRangedFloat(Minimum=0.85, Maximum=1.25)
    Mesh Scale=&Multiply_VectorByFloat(
     Vector=&MakeVector(X=0.6, Y=0.5, Z=&RandomRangeFloat(Minimum=0.7, Maximum=1))
     Float=&RandomRangeFloat(Minimum=1.3, Maximum=1.3)
    )
    Mesh Scale Mode=::NewEnumerator4
    Sprite Size=&Vector2DFromFloat(Value=50)
    Sprite Size Mode=::NewEnumerator4
    UsePositionOffset=false
    Write Material Random=true
   ) {module=InitializeParticle001}
   /SetVariables(Particles.Burned:NiagaraFloat=&ClampFloat(Float=&FloatFromCurve(FloatCurve=DataInterface() {propertyValues {Curve {keys [
          {arriveTangent=1.1829, interpMode=RCIM_Cubic, leaveTangent=1.1829, tangentMode=RCTM_User, time=0.5, value=0}
          {interpMode=RCIM_Cubic, tangentMode=RCTM_User, time=1, value=1}
         ]}}}))) {isSetParameters=true}
   /SetVariables(Particles.Size:NiagaraFloat=&Multiply_Float(
     A=1
     B=&FloatFromCurve(
      FloatCurve=DataInterface() {propertyValues {Curve {keys [{interpMode=RCIM_Cubic, tangentMode=RCTM_User, time=0, value=1}, {interpMode=RCIM_Cubic, tangentMode=RCTM_User, time=1, value=0.8}]}}}
      Scale Curve=0.5
     )
    )) {isSetParameters=true}
   /SetVariables(
    Particles.Lifetime=&self:ScratchDynamicInput(Nth:NiagaraInt32=12, ScaleFactor:NiagaraFloat=&RandomRangeFloat(Minimum=1.1, Maximum=1.3))
    Particles.Scale=&self:Scale_Nth_IDx(
     Nth:NiagaraInt32=12
     Scale:Vector3f=$Particles.Scale
     ScaleFactor:NiagaraFloat=&RandomRangeFloat(Minimum=1.1, Maximum=1.3)
    )
   ) {enabled=false, isSetParameters=true}
   /SetVariables(Particles.Magnitude:NiagaraFloat=&RemapRange(Input Value To Remap=$Particles.Burned, Input Min=0, Input Max=1, Output Min=1, Output Max=0.1)) {isSetParameters=true}
   &self:ScratchModule_03(Clip=0.7, Nth=15) {enabled=false, module=ScratchModule_03}
   /SetVariables(Particles.Lifetime=0.5) {isSetParameters=true}
   &ShapeLocation(MaskSpawnGroup=false, Override Seed=false, Ring Radius=4, Shape Primitive=::NewEnumerator4, UseOwnerScale=true)
   &CameraOffset(Camera Offset Amount=&RandomRangeFloat(Minimum=10, Maximum=25))
  ]
  particleUpdateScript [
   &ParticleState()
   &SolveForcesAndVelocity(Clamp Velocity=false, Limit Acceleration=false)
   &self:Fade_02(
    AgeScale=1
    Fade Curve=DataInterface() {propertyValues {Curve {keys [
        {arriveTangent=16.5612, interpMode=RCIM_Cubic, leaveTangent=16.5612, tangentMode=RCTM_User, time=0, value=0}
        {interpMode=RCIM_Cubic, tangentMode=RCTM_User, time=0.1, value=1}
        {interpMode=RCIM_Cubic, tangentMode=RCTM_User, time=0.5, value=1}
        {interpMode=RCIM_Cubic, tangentMode=RCTM_User, time=1, value=0.35}
       ]}}}
   ) {enabled=false, module=Fade_02}
   &ScaleColor(
    Scale Alpha=&FloatFromCurve(FloatCurve=DataInterface() {propertyValues {Curve {keys [
         {arriveTangent=9.0725, interpMode=RCIM_Cubic, leaveTangent=9.0725, tangentMode=RCTM_User, time=0, value=0}
         {arriveTangent=0.1459, interpMode=RCIM_Cubic, leaveTangent=0.1459, tangentMode=RCTM_User, time=0.034, value=0.9847}
         {interpMode=RCIM_Cubic, tangentMode=RCTM_User, time=0.4, value=1}
         {arriveTangent=-0.0405, interpMode=RCIM_Cubic, leaveTangent=-0.0405, tangentMode=RCTM_User, time=1, value=0.1}
        ]}}})
    Scale RGB=&Multiply_VectorByFloat(
     Vector=<1 1 1>
     Float=&FloatFromCurve(FloatCurve=DataInterface() {propertyValues {Curve {keys [
          {arriveTangent=9.0725, interpMode=RCIM_Cubic, leaveTangent=9.0725, tangentMode=RCTM_User, time=0, value=0}
          {arriveTangent=-2.5249, interpMode=RCIM_Cubic, leaveTangent=-2.5249, tangentMode=RCTM_User, time=0.0279, value=0.9847}
          {arriveTangent=-0.0405, interpMode=RCIM_Cubic, leaveTangent=-0.0405, tangentMode=RCTM_User, time=1, value=0.1}
         ]}}})
    )
    ScaleA=true
    ScaleRGB=true
   )
   &ScaleSpriteSize(
    Non-Uniform Curve Sprite Scale=DataInterface() {propertyValues {
      XCurve {keys [{tangentMode=RCTM_User, time=0, value=1}, {tangentMode=RCTM_User, time=1, value=1}]}
      YCurve {keys [{tangentMode=RCTM_User, time=0, value=1}, {tangentMode=RCTM_User, time=1, value=1}]}
     }}
    Scale Sprite Size Mode=::NewEnumerator3
   ) {enabled=false}
   &SpriteFacingAndAlignment(
    Sprite Alignment=&Lerp_Vector(A=$Particles.Velocity, B=<0 0 1>, Alpha=0.5)
    WriteAlignment=true
    WriteFacing=false
   ) {enabled=false}
   &ScaleMeshSize(Scale Factor=&Multiply_VectorByFloat(
     Vector=$Particles.Initial.Scale
     Float=&FloatFromCurve(
      FloatCurve=DataInterface() {propertyValues {Curve {keys [
          <at 0 0.7 RCIM_Cubic>
          {arriveTangent=-0.565, interpMode=RCIM_Cubic, leaveTangent=-0.565, tangentMode=RCTM_User, time=0.2, value=1}
          {interpMode=RCIM_Cubic, tangentMode=RCTM_User, time=1, value=0.75}
         ]}}}
      Scale Curve=&Multiply_Float(A=$Particles.Size, B=1)
     )
    )) {module=ScaleMeshSize001}
   &SubUVAnimation(
    Start Frame:NiagaraInt32=&RandomRangeInt(Minimum=0, Maximum=10)
    SubUV Loop Count:NiagaraFloat=&RandomRangeFloat(Minimum=0.3, Maximum=0.5)
   )
   &DynamicMaterialParameters(
    Index 0 Param 1=&FloatFromCurve(FloatCurve=DataInterface() {propertyValues {Curve {keys [<at 0 1>,<at 1 0>]}}})
    Index 0 Param 2=&RandomRangeFloat(Minimum=0, Maximum=2)
    Index 0 Param 3=8
    Index 0 Param 4=$Particles.SubImageIndex
    Index 1 Param 1=&RandomRangeFloat(Minimum=-1, Maximum=1)
    Index 1 Param 2=&RandomRangeFloat(Minimum=0, Maximum=3)
    Index 1 Param 3=&FloatFromCurve(
     FloatCurve=DataInterface() {propertyValues {Curve {keys [
         {arriveTangent=4.1785, interpMode=RCIM_Cubic, leaveTangent=4.1785, tangentMode=RCTM_User, time=0, value=0}
         {arriveTangent=0.9995, interpMode=RCIM_Cubic, leaveTangent=0.9995, tangentMode=RCTM_User, time=0.2223, value=0.6825}
         {arriveTangent=-0.0405, interpMode=RCIM_Cubic, leaveTangent=-0.0405, tangentMode=RCTM_User, time=0.7, value=0.9}
         <at 1 0.25>
        ]}}}
     Scale Curve=&RemapRange(Input Value To Remap=$Particles.Burned, Input Min=0, Input Max=1, Output Min=3, Output Max=0.5)
    )
    Module.Param0WriteEnabled=true
    Module.Param0WriteEnabled1=true
    Module.Param1WriteEnabled=true
    Module.Param1WriteEnabled1=true
    Module.Param2WriteEnabled=true
    Module.Param2WriteEnabled1=true
    Module.Param3WriteEnabled=true
    Module.Param3WriteEnabled1=false
    Write Parameter Index 1=true
   )
   &Color(Color=<rgba 0 0.1429 1 1>) {enabled=false}
  ]
  rendererClasses [&NiagaraMeshRendererProperties]
  renderers [
   {rendererClass=&NiagaraMeshRendererProperties, rendererIndex=0}
   {rendererClass=&NiagaraMeshRendererProperties, rendererIndex=1}
  ]
 }
}
summary {
 systemName=NS_VictoryHighway_GroundImpact
 emitters [
  {
   emitterName=HT_Debris
   simTarget=GPUComputeSim
   rendererClasses [&NiagaraSpriteRendererProperties, &NiagaraMeshRendererProperties]
  }
  {emitterName=GroundDust, rendererClasses [&NiagaraSpriteRendererProperties]}
  {
   emitterName=Ground_Cracks001
   rendererClasses [&NiagaraDecalRendererProperties, &NiagaraSpriteRendererProperties]
  }
  {emitterName=LavaSplash001, rendererClasses [&NiagaraSpriteRendererProperties]}
  {emitterName=Embers, rendererClasses [&NiagaraSpriteRendererProperties]}
  {emitterName=Flame_Streaks, rendererClasses [&NiagaraSpriteRendererProperties]}
  {emitterName=Shockwave, rendererClasses [&NiagaraMeshRendererProperties]}
  {bEnabled=false, emitterName=BlowingParticles, rendererClasses [&NiagaraDecalRendererProperties]}
  {bEnabled=false, emitterName=FlameBeam, rendererClasses [&NiagaraSpriteRendererProperties]}
  {emitterName=RingBurst, rendererClasses [&NiagaraSpriteRendererProperties]}
  {emitterName=Voxel_On_Med001, rendererClasses [&NiagaraMeshRendererProperties]}
 ]
 userVariables [
  User.bIsFrontEnd {type=NiagaraBool}
  User.bIsFrontEndPreview {type=NiagaraBool}
  User.DesertBoulder {type=NiagaraBool}
  User.ImpactColor {type=LinearColor, description="", defaultValue {struct=LinearColor, value=<rgba 1 1 1 1>}}
  User.IsWeakPoint {type=NiagaraBool}
  User.SourcePlaneSize {type=Vector3f, description="", defaultValue {struct=Vector3f, value=<50 50 1>}}
 ]
}
