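# Niagara system for the VictoryHighway emote's hand fire, spawned on skeletal mesh samples from a system-level mesh.
# Flipbook fire, GPU embers with curl noise and per-distance spawning, and a ribbon flame trail form the flames.
# Smoke bursts with flare sprites, tapering falling drops, and multi-renderer glow sprites that follow the mesh add detail.
# Embers and trail blend between world and local space by owner speed; embers and glow also use near-camera fading.
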NiagaraSystem /BRCosmetics/Animation/Game/MainPlayer/Emotes/VictoryHighway/VFX/NS_VictoryHighway_HandFire
dependencies {
 usedDataInterfaces [
  &NiagaraDataInterfaceCurve
  &NiagaraDataInterfaceColorCurve
  &NiagaraDataInterfaceSpriteRendererInfo
  &NiagaraDataInterfaceArrayInt32
  &NiagaraDataInterfaceVector2DCurve
  &NiagaraDataInterfaceSkeletalMesh
  &NiagaraDataInterfaceCamera
 ]
 usedDynamicInputs [
  &FN_LOD_Distance_Remap
  &MakeFloatFromVector4
  &ReturnCameraVector
 ]
 usedModules [
  &SystemState
  &BlendSpaceHelper
  &self:SystemSpawnScript.SystemScriptSource.SystemScriptGraph.NiagaraNodeAssignment_4.SetVariables_#1
  &self:SystemSpawnScript.SystemScriptSource.SystemScriptGraph.NiagaraNodeAssignment_2.SetVariables_#2
  &self:SystemSpawnScript.SystemScriptSource.SystemScriptGraph.NiagaraNodeAssignment_3.SetVariables_#3
  &self:SystemSpawnScript.SystemScriptSource.SystemScriptGraph.NiagaraNodeAssignment_1.SetVariables_#4
  &self:SystemSpawnScript.SystemScriptSource.SystemScriptGraph.NiagaraNodeAssignment_0.SetVariables_#5
 ]
}
emitters {
 Drops {
  emitterUpdateScript [
   &EmitterState(Life Cycle Mode=::NewEnumerator1, Loop Behavior=::NewEnumerator1, Loop Delay=1, Loop Duration=0.5, UseLoopDelay=true)
   &SpawnRate(
    SpawnRate=&FloatFromCurve(FloatCurve=DataInterface() {propertyValues {Curve {keys [<at 0 1>,<at 1 0>]}}}, CurveIndex=$Emitter.NormalizedLoopAge, Scale Curve=200)
    Use Spawn Probability=false
   )
  ]
  particleSpawnScript [
   /SetVariables(Particles.Random=&UniformRangedFloat(Minimum=0, Maximum=1)) {isSetParameters=true}
   &InitializeParticle(
    Color Maximum=<rgba 1 0.1147 0 1>
    Color Minimum=<rgba 1 0.2794 0.0594 1>
    Color Mode=::NewEnumerator2
    Lifetime=&UniformRangedFloat(Minimum=0.2, Maximum=0.8)
    Mass=&UniformRangedFloat(Minimum=0.6, Maximum=1.2)
    Mass Mode=::NewEnumerator1
    Sprite Rotation Mode=::NewEnumerator0
    Sprite Size=&Vector2DFromCurve(
     Vector2Curve=DataInterface() {propertyValues {
       XCurve {keys [<at -0.0011 2.0768 RCIM_Cubic>,<at 0.9989 0.3489 RCIM_Cubic>]}
       YCurve {keys [
         {arriveTangent=-30.8979, interpMode=RCIM_Cubic, leaveTangent=-30.8979, tangentMode=RCTM_User, time=0, value=11}
         {arriveTangent=-9.0776, interpMode=RCIM_Cubic, leaveTangent=-9.0776, time=0.2382, value=6.3246}
         <at 0.9989 1.9328 RCIM_Cubic>
        ]}
      }}
     CurveIndex=$Particles.Random
     ScaleCurve=<1.2 1.2>
    )
    Sprite Size Mode=::NewEnumerator4
    UsePositionOffset=false
    Write Material Random=true
   ) {module=InitializeParticle001}
   &SkeletalMeshLocation(Bone Sampling Mode=::NewEnumerator4, MaskSpawnGroup=false, Orientation Sampling=::NewEnumerator1, Skeletal Mesh=$System.SK_Mesh)
   &DynamicMaterialParameters(
    Index 0 Param 1=&UniformRangedFloat(Minimum=4, Maximum=10)
    Module.Param0WriteEnabled=true
    Module.Param1WriteEnabled=false
    Module.Param2WriteEnabled=false
    Module.Param3WriteEnabled=false
   )
   &InheritVelocity(Module.Limit Speed=true, Velocity Limit=800, Velocity Scale=<0.3 0.3 0>)
   &AddVelocity(Coordinate Space=::World, Velocity=&UniformRangedVector(Minimum=<-100 -100 100>, Maximum=<100 100 300>))
  ]
  particleUpdateScript [
   &ParticleState()
   /Game/Effects/Niagara/Emitters/Water/FN_DistanceKillParticles(AgeReduiceAmount=0.01)
   &ScaleColor(
    Scale Alpha=&FloatFromCurve(FloatCurve=DataInterface() {propertyValues {Curve {keys [<at 0 0>,<at 0.15 1>,<at 1 0>]}}})
    Scale RGB=&VectorFromFloat(Value=&FloatFromCurve(FloatCurve=DataInterface() {propertyValues {Curve {keys [<at 0 2>,<at 1 0>]}}}))
    ScaleA=true
    ScaleRGB=true
   )
   &Drag(Use Linear Drag=true, Use Rotational Drag=false)
   &GravityForce()
   &SolveForcesAndVelocity(Clamp Velocity=false, Limit Acceleration=false)
  ]
  rendererClasses [&NiagaraSpriteRendererProperties]
  renderers [{rendererClass=&NiagaraSpriteRendererProperties}]
 }
 Embers_Burst {
  simTarget=GPUComputeSim
  emitterUpdateScript [
   &EmitterState()
   &SpawnRate(
    SpawnRate=&Multiply_Float(
     A=30
     B=&FloatFromCurve(
      FloatCurve=DataInterface() {propertyValues {Curve {keys [
          {arriveTangent=-4.4442, interpMode=RCIM_Cubic, leaveTangent=-4.4442, leaveTangentWeight=0.8941, tangentMode=RCTM_User, time=0.2, value=1}
          {arriveTangentWeight=0.9632, interpMode=RCIM_Cubic, tangentMode=RCTM_User, time=0.8, value=0}
         ]}}}
      CurveIndex=$Emitter.NormalizedLoopAge
     )
    )
    Use Spawn Probability=false
   ) {enabled=false}
   &SpawnPerUnit(Module.Use Max Movement Threshold=false, Module.Use Movement Tolerance=false, Spawn Spacing=5, Use Spawn Probability=false)
   &SpawnBurst_Instantaneous(Spawn Count=30, Use Spawn Probability=false, UseLoopCountLimit=false)
  ]
  particleSpawnScript [
   &InitializeParticle(
    Color=<rgba 1 1 1 2>
    Lifetime=&UniformRangedFloat(Minimum=0.4, Maximum=0.8)
    Mass Mode=::NewEnumerator1
    Position=&SimulationPosition()
    Position Mode=::NewEnumerator1
    Sprite Size=&Vector2DFromFloat(Value=&FloatFromCurve(
      FloatCurve=DataInterface() {propertyValues {Curve {keys [<at 0 1.2>,<at 0.1257 0.9847>,<at 0.2996 0.5>,<at 1 0.3>]}}}
      CurveIndex=&UniformRangedFloat(Minimum=0, Maximum=1)
      Scale Curve=2
     ))
    Sprite Size Mode=::NewEnumerator4
    UsePositionOffset=false
    Write Material Random=true
   ) {module=InitializeParticle001}
   &SkeletalMeshLocation(Bone Sampling Mode=::NewEnumerator4, MaskSpawnGroup=false, Orientation Sampling=::NewEnumerator1, Skeletal Mesh=$System.SK_Mesh)
   &ShapeLocation(MaskSpawnGroup=false, Override Seed=false, Sphere Radius=5, UseOwnerScale=true)
   &AddVelocity(
    Additional Quaternion Rotation=false
    Cone Angle=&FloatFromCurve(FloatCurve=DataInterface() {propertyValues {Curve {keys [<at 0 1>,<at 1 0.05>]}}}, CurveIndex=$Emitter.Age, Scale Curve=20)
    Module.UseConeFalloff=false
    Module.UseVelDistribution=false
    Override Seed=false
    Rotation Mode=::NewEnumerator3
    Rotation Quaternion:Quat4f=$Particles.SkeletalMeshLocation.SampledRotation
    Velocity Mode=::NewEnumerator2
    Velocity Speed=&Multiply_Float(
     A=&Multiply_Float(
      A=&RandomRangeFloat(
       Minimum=300
       Maximum=&FloatFromCurve(
        FloatCurve=DataInterface() {propertyValues {Curve {keys [{tangentMode=RCTM_User, time=0, value=1}, <at 0.9 0.15>]}}}
        CurveIndex=$Emitter.Age
        Scale Curve=2000
       )
      )
      B=1
     )
     B=$System.FloatCurveRampeDownBySystemAge
    )
   ) {enabled=false}
   &InheritVelocity(
    Inherited Velocity Amount Scale=&Multiply_VectorByFloat(Vector=<0.5 0.5 0.5>, Float=$System.FloatCurveRampeDownBySystemAge)
    Speed Limit=false
    UseSpeedThreshold=false
   )
  ]
  particleUpdateScript [
   &ParticleState()
   &ScaleSpriteSize(Scale Sprite Size Mode=::NewEnumerator0, Uniform Scale Factor=$System.DistanceScalar)
   &Drag(
    Drag=&Multiply_Float(
     A=&FloatFromCurve(
      FloatCurve=DataInterface() {propertyValues {Curve {keys [<at 0 0.5>, {arriveTangent=0.5, interpMode=RCIM_Cubic, leaveTangent=0.5, time=0.3, value=0.5}, <at 1 1 RCIM_Cubic>]}}}
      Scale Curve=12
     )
     B=1
    )
    Use Linear Drag=true
    Use Rotational Drag=false
   )
   /Niagara/Modules/Update/Forces/CurlNoiseForce(
    Noise Frequency=10
    Noise Strength=&FloatFromCurve(FloatCurve=DataInterface() {propertyValues {Curve {keys [<at 0 1>,<at 1 0>]}}}, Scale Curve=3000)
    Pan Noise=true
    Pan Noise Field=<0.1 0.1 0.5>
   )
   &GravityForce(Gravity=&RandomRangeVector(Minimum=<0 0 500>, Maximum=<0 0 -500>))
   &SolveForcesAndVelocity(Clamp Velocity=false, Limit Acceleration=false)
   &ScaleSpriteSizeBySpeed(
    Max Scale Factor=<0.8 5>
    Min Scale Factor=<0.8 0.8>
    Sample Scale Factor By Curve=true
    Scale Factor Curve=DataInterface() {propertyValues {Curve {keys [
        <at -0.0077 0.0765 RCIM_Cubic>
        {arriveTangent=1.2463, interpMode=RCIM_Cubic, leaveTangent=1.2463, tangentMode=RCTM_User, time=0.9985, value=0.9549}
       ]}}}
    Velocity Threshold=300
   )
   &ScaleColor(
    Scale Alpha=&FloatFromCurve(FloatCurve=DataInterface() {propertyValues {Curve {keys [<at 0.25 1>,<at 1 0>]}}})
    Scale RGB=&MakeVectorFromLinearColorRGB(Color=&ColorFromCurve(ColorCurve=DataInterface() {propertyValues {
        AlphaCurve {keys [<at 0 1 RCIM_Cubic>,<at 0.7107 1>]}
        BlueCurve {keys [<at 0 14.6701 RCIM_Cubic>,<at 0.368 3.776>,<at 0.855 0.1753 RCIM_Cubic>]}
        GreenCurve {keys [<at 0 36.7481 RCIM_Cubic>,<at 0.368 13.2325>,<at 0.855 0.3243 RCIM_Cubic>]}
        RedCurve {keys [<at 0 50 RCIM_Cubic>,<at 0.368 25>,<at 0.855 1 RCIM_Cubic>]}
       }}))
    ScaleA=true
    ScaleRGB=true
   )
   &ScaleColor(
    Scale Alpha=&FloatFromCurve(
     FloatCurve=DataInterface() {propertyValues {Curve {keys [
         {arriveTangent=1.5243, interpMode=RCIM_Cubic, leaveTangent=1.5243, tangentMode=RCTM_User, time=0, value=0.1}
         {arriveTangent=0.0515, interpMode=RCIM_Cubic, leaveTangent=0.0515, tangentMode=RCTM_User, time=1, value=1}
        ]}}}
     CurveIndex=$System.NearCameraFade
    )
    ScaleA=true
    ScaleRGB=true
   ) {module=ScaleColor001}
   &BlendSpace("Position Blend (World <-> Local)":NiagaraFloat=&MakeCustomFloatFromBool(
     Bool=&SetBoolByFloatComparison(A=&VectorLength(Vector=$Engine.Owner.Velocity), B=400)
     True Float=0.8
     False Float=0
     Translation Contribution:NiagaraFloat=1
     Rotation Contribution:NiagaraFloat=1
    ))
  ]
  rendererClasses [&NiagaraSpriteRendererProperties]
  renderers [{rendererClass=&NiagaraSpriteRendererProperties}]
 }
 Fire_Burst {
  emitterUpdateScript [
   &EmitterState()
   &SpawnBurst_Instantaneous(Spawn Count=6, Spawn Group=1, Use Spawn Probability=false, UseLoopCountLimit=false)
   &SpawnRate(SpawnRate=100, Use Spawn Probability=true)
  ]
  particleSpawnScript [
   &InitializeParticle(
    Color=<rgba 5 0.6569 0.0985 1>
    Lifetime=&UniformRangedFloat(Minimum=0.4, Maximum=0.5)
    Mass Mode=::NewEnumerator1
    Position=&SimulationPosition()
    Position Mode=::NewEnumerator1
    Sprite Rotation Mode=::NewEnumerator2
    Sprite Size=&Vector2DFromFloat(Value=&FloatFromCurve(
      FloatCurve=DataInterface() {propertyValues {Curve {keys [<at 0 1.2>,<at 0.1257 0.9847>,<at 0.2996 0.5>,<at 1 0.3>]}}}
      CurveIndex=&UniformRangedFloat(
       Minimum=&MaskFloat(Mask Source=1, Mask Parameter=$Emitter.SpawnGroup, Mask=0, Masked Float=0, Passthrough Float=0)
       Maximum=&MaskFloat(Mask Source=1, Mask Parameter=$Emitter.SpawnGroup, Mask=0, Masked Float=1, Passthrough Float=0.3)
      )
      Scale Curve=&MaskFloat(Mask Source=1, Mask Parameter=$Emitter.SpawnGroup, Mask=0, Masked Float=2, Passthrough Float=5)
     ))
    Sprite Size Mode=::NewEnumerator4
    Sprite UV Mode=::NewEnumerator3
    UsePositionOffset=true
    Write Material Random=true
   ) {module=InitializeParticle001}
   &SkeletalMeshLocation(
    Bone Sampling Mode=::NewEnumerator4
    MaskSpawnGroup=false
    Orientation Sampling=::NewEnumerator1
    Sampled Position Offset=&MaskVector(Mask Source=1, Mask Parameter=$Emitter.SpawnGroup, Mask=0, Masked Vector=<-5 0 0>, Passthrough Vector=<-5 0 0>)
    Skeletal Mesh=$System.SK_Mesh
   )
   &ShapeLocation(MaskSpawnGroup=false, Override Seed=false, Sphere Radius=5, UseOwnerScale=true)
   &InheritVelocity(
    Inherited Velocity Amount Scale=&Multiply_VectorByFloat(Vector=<0.6 0.6 0.6>, Float=$System.FloatCurveRampeDownBySystemAge)
    Speed Limit=false
    UseSpeedThreshold=false
   )
   &DynamicMaterialParameters(
    Index 0 Param 2=&RandomRangeFloat(Minimum=0, Maximum=4)
    Index 0 Param 4=2.5
    Module.Param0WriteEnabled=false
    Module.Param1WriteEnabled=true
    Module.Param2WriteEnabled=false
    Module.Param3WriteEnabled=true
   ) {module=DynamicMaterialParameters001}
   &CameraOffset(Camera Offset Amount=&MaskFloat(
     Mask Source=1
     Mask Parameter=$Emitter.SpawnGroup
     Mask=0
     Masked Float=0
     Passthrough Float=&RandomRangeFloat(Minimum=50, Maximum=15)
    ))
  ]
  particleUpdateScript [
   &ParticleState()
   &SubUVAnimation(
    "Sprite Renderer "=DataInterface() {propertyValues {SpriteRenderer=&self:Fire_Burst_4.NiagaraSpriteRendererProperties_2}}
    UseEndFrame=false
    UseStartFrame=false
   ) {module=SubUVAnimation001}
   &ScaleSpriteSize(
    Scale Sprite Size Mode=::NewEnumerator0
    Uniform Scale Factor=&Multiply_Float(
     A=&FloatFromCurve(FloatCurve=DataInterface() {propertyValues {Curve {keys [<at 0 1.375 RCIM_Cubic>, {arriveTangent=1.25, interpMode=RCIM_Cubic, leaveTangent=1.25, time=0.6, value=1.625}, <at 1 2.625 RCIM_Cubic>]}}})
     B=8
    )
   )
   &Drag(
    Drag=&Multiply_Float(
     A=&FloatFromCurve(
      FloatCurve=DataInterface() {propertyValues {Curve {keys [<at 0 0.5>, {arriveTangent=0.5, interpMode=RCIM_Cubic, leaveTangent=0.5, time=0.3, value=0.5}, <at 1 1 RCIM_Cubic>]}}}
      Scale Curve=12
     )
     B=1
    )
    Use Linear Drag=true
    Use Rotational Drag=false
   )
   &AccelerationForce(Acceleration=<0 0 50>)
   &SolveForcesAndVelocity(Clamp Velocity=false, Limit Acceleration=false)
   &DynamicMaterialParameters(
    Index 0 Param 1=&FloatFromCurve(FloatCurve=DataInterface() {propertyValues {Curve {keys [<at 0 1>]}}})
    Index 0 Param 2=0.5
    Index 0 Param 3=&FloatFromCurve(FloatCurve=DataInterface() {propertyValues {Curve {keys [<at 0 0.5 RCIM_Cubic>, {arriveTangent=1.5, interpMode=RCIM_Cubic, leaveTangent=1.5, time=0.5235, value=0.5}, <at 1 2>]}}})
    Module.Param0WriteEnabled=true
    Module.Param1WriteEnabled=false
    Module.Param2WriteEnabled=true
    Module.Param3WriteEnabled=false
   )
   &ScaleColor(
    Scale Alpha=&Multiply_Float(
     A=0.8
     B=&FloatFromCurve(FloatCurve=DataInterface() {propertyValues {Curve {keys [
          {arriveTangent=0.162, interpMode=RCIM_Cubic, leaveTangent=0.162, tangentMode=RCTM_User, time=0, value=0.1}
          {arriveTangent=1.5, interpMode=RCIM_Cubic, leaveTangent=1.5, time=0.1, value=0.9}
          {arriveTangent=-1.6513, interpMode=RCIM_Cubic, leaveTangent=-1.6513, tangentMode=RCTM_User, time=0.3, value=0.55}
          {arriveTangent=-0.4875, interpMode=RCIM_Cubic, leaveTangent=-0.4875, tangentMode=RCTM_User, time=0.5, value=0.15}
          {interpMode=RCIM_Cubic, leaveTangent=-2.371, time=1, value=0}
         ]}}})
    )
    Scale RGB=&VectorFromFloat(Value=&FloatFromCurve(
      FloatCurve=DataInterface() {propertyValues {Curve {keys [<at 0 1 RCIM_Cubic>, {arriveTangent=-1, interpMode=RCIM_Cubic, leaveTangent=-1, time=0.2, value=0.225}, <at 1 0 RCIM_Cubic>]}}}
      Scale Curve=20
     ))
    ScaleA=true
    ScaleRGB=true
   ) {module=ScaleColor001}
   &BlendSpace("Position Blend (World <-> Local)":NiagaraFloat=&MakeCustomFloatFromBool(
     Bool=&SetBoolByFloatComparison(A=&VectorLength(Vector=$Engine.Owner.Velocity), B=400)
     True Float=0.8
     False Float=0
     Translation Contribution:NiagaraFloat=1
     Rotation Contribution:NiagaraFloat=1
    )) {enabled=false}
  ]
  rendererClasses [&NiagaraSpriteRendererProperties]
  renderers [{rendererClass=&NiagaraSpriteRendererProperties}]
 }
 Flametrail {
  emitterUpdateScript [
   &EmitterState(Inactive Response=::NewEnumerator1, Life Cycle Mode=::NewEnumerator1, Loop Behavior=::NewEnumerator1, Loop Duration=1.5, UseLoopDelay=false)
   &SpawnRate(SpawnRate=30, Use Spawn Probability=false)
  ]
  particleSpawnScript [
   &InitializeParticle(Ribbon Width=3, Ribbon Width Mode=::NewEnumerator1, UsePositionOffset=false, Write Material Random=true)
   &SkeletalMeshLocation(Bone Sampling Mode=::NewEnumerator4, MaskSpawnGroup=false, Orientation Sampling=::NewEnumerator1, Skeletal Mesh=$System.SK_Mesh)
   &AddVelocity(Velocity=<0 0 200>) {enabled=false}
  ]
  particleUpdateScript [
   &GravityForce(Gravity=<0 0 25>)
   &ParticleState()
   &ScaleColor(
    Linear Color Curve=DataInterface() {propertyValues {
      AlphaCurve {keys [<at 0.0136 0>,<at 0.0647 1>,<at 0.5518 0.8399>,<at 1 0>]}
      BlueCurve {keys [<at 0 0>,<at 0.3089 1>]}
      GreenCurve {keys [<at 0 0.4343>,<at 0.3089 1>]}
      RedCurve {keys [<at 0 2>,<at 0.3089 1>]}
     }}
    Scale Mode=::NewEnumerator2
   ) {module=ScaleColor001}
   &ScaleColor(
    Scale Alpha=&FloatFromCurve(
     FloatCurve=DataInterface() {propertyValues {Curve {keys [{interpMode=RCIM_Cubic, tangentMode=RCTM_User, time=0, value=1}, <at 1 0>]}}}
     Scale Curve=&MakeCustomFloatFromBool(
      Bool=&SetBoolByFloatComparison(A=&VectorLength(Vector=$Engine.Owner.Velocity), B=300)
      True Float=&FloatFromCurve(
       FloatCurve=DataInterface() {propertyValues {Curve {keys [
           <at 0.1103 0>
           {arriveTangent=-0.0688, interpMode=RCIM_Cubic, leaveTangent=-0.0688, tangentMode=RCTM_User, time=0.2644, value=1}
           {arriveTangent=-2.9959, interpMode=RCIM_Cubic, leaveTangent=-2.9959, tangentMode=RCTM_User, time=1, value=0}
          ]}}}
       CurveIndex=$Emitter.NormalizedLoopAge
      )
      False Float=&FloatFromCurve(
       FloatCurve=DataInterface() {propertyValues {Curve {keys [
           <at 0.1103 0>
           {arriveTangent=-0.0688, interpMode=RCIM_Cubic, leaveTangent=-0.0688, tangentMode=RCTM_User, time=0.2644, value=1}
           {arriveTangent=-2.9959, interpMode=RCIM_Cubic, leaveTangent=-2.9959, tangentMode=RCTM_User, time=1, value=0}
          ]}}}
       CurveIndex=$Emitter.NormalizedLoopAge
      )
     )
    )
    Scale RGB=&VectorFromFloat(Value=0.4)
    ScaleA=true
    ScaleRGB=true
   )
   &ScaleRibbonWidth(Ribbon Width Scale=&FloatFromCurve(
     FloatCurve=DataInterface() {propertyValues {Curve {keys [<at 0 1 RCIM_Cubic>, {arriveTangent=-1.7794, interpMode=RCIM_Cubic, leaveTangent=-1.7794, tangentMode=RCTM_User, time=1, value=0}]}}}
     Scale Curve=2
    ))
   &DynamicMaterialParameters(
    Index 0 Param 3=1
    Module.Index0Float0Write=true
    Module.Index0Float1Write=true
    Module.Index0Float2Write=true
    Module.Index0Float3Write=true
   )
   &SolveForcesAndVelocity(Clamp Velocity=false, Limit Acceleration=false)
   &BlendSpace("Position Blend (World <-> Local)":NiagaraFloat=&MakeCustomFloatFromBool(
     Bool=&SetBoolByFloatComparison(A=&VectorLength(Vector=$Engine.Owner.Velocity), B=300)
     True Float=0.5
     False Float=0
     Translation Contribution:NiagaraFloat=1
     Rotation Contribution:NiagaraFloat=1
    ))
  ]
  rendererClasses [&NiagaraRibbonRendererProperties]
  renderers [{rendererClass=&NiagaraRibbonRendererProperties}]
 }
 ParticleDecal {
  emitterUpdateScript [
   &EmitterState(
    Inactive Response=::NewEnumerator1
    Life Cycle Mode=::NewEnumerator1
    Loop Behavior=::NewEnumerator1
    Loop Duration Mode=::NewEnumerator1
    UseLoopDelay=false
   )
   &SpawnBurst_Instantaneous(Use Spawn Probability=true, UseLoopCountLimit=false)
  ]
  particleSpawnScript [
   &PartitionParticles(Number Of Partitions=26)
   &InitializeParticle(
    Color=<rgba 1 0.2549 0.0234 1>
    Lifetime=0.6
    Sprite Rotation Mode=::NewEnumerator2
    Sprite Size Mode=::NewEnumerator3
    Uniform Sprite Size=&Multiply_Float(A=30, B=1)
    UsePositionOffset=false
    Write Material Random=true
   )
   &CameraOffset(Camera Offset Amount=20) {module=CameraOffset001}
   &DynamicMaterialParameters(
    Index 0 Param 1=1
    Index 0 Param 2=1
    Module.Index0Param0Write=true
    Module.Index0Param1Write=true
    Module.Index0Param2Write=true
    Module.Index0Param3Write=false
   )
   /SetVariables(Particles.EnergyDP:Vector4f=<2 1 8 1>) {isSetParameters=true}
  ]
  particleUpdateScript [
   &ParticleState(Kill Particles When Lifetime Has Elapsed=false, Let Infinitely Lived Particles Die When Emitter Deactivates=true)
   &ScaleColor(
    Scale Mode=::NewEnumerator1
    Scale RGBA=&MakeVector4_FromLinearColor(LinearColor=<rgba 1 0.3998 0.1409 1>)
    ScaleRGBA=true
   ) {module=ScaleColor001}
   &ScaleColor(
    Scale Alpha=&Multiply_Float(
     A=1
     B=&FloatFromCurve(
      FloatCurve=DataInterface() {propertyValues {Curve {keys [
          {arriveTangent=1.5243, interpMode=RCIM_Cubic, leaveTangent=1.5243, tangentMode=RCTM_User, time=0, value=0.1}
          {arriveTangent=0.0515, interpMode=RCIM_Cubic, leaveTangent=0.0515, tangentMode=RCTM_User, time=1, value=1}
         ]}}}
      CurveIndex=$System.NearCameraFade
      Scale Curve=&FloatFromCurve(
       FloatCurve=DataInterface() {propertyValues {Curve {keys [<at 0 1>, {interpMode=RCIM_Cubic, tangentMode=RCTM_User, time=0.8, value=1}, {arriveTangent=-8.2155, tangentMode=RCTM_User, time=1, value=0}]}}}
       CurveIndex=$System.NormalizedLoopAge
      )
     )
    )
    Scale RGB=&VectorFromFloat(Value=5)
    ScaleA=true
    ScaleRGB=true
   ) {module=ScaleColor002}
   &ScaleColor(
    Scale Alpha=&Multiply_Float(
     A=&FloatFromCurve(FloatCurve=DataInterface() {propertyValues {Curve {keys [
          {arriveTangent=0.3051, interpMode=RCIM_Cubic, leaveTangent=0.3051, tangentMode=RCTM_User, time=0, value=1}
          {arriveTangent=0.3995, interpMode=RCIM_Cubic, leaveTangent=0.3995, tangentMode=RCTM_User, time=0.1408, value=0.9847}
          {arriveTangent=-0.0405, interpMode=RCIM_Cubic, leaveTangent=-0.0405, tangentMode=RCTM_User, time=1, value=0}
         ]}}})
     B=&FloatFromCurve(
      FloatCurve=DataInterface() {propertyValues {Curve {keys [
          {arriveTangent=1.5243, interpMode=RCIM_Cubic, leaveTangent=1.5243, tangentMode=RCTM_User, time=0, value=0.1}
          {arriveTangent=0.0515, interpMode=RCIM_Cubic, leaveTangent=0.0515, tangentMode=RCTM_User, time=1, value=1}
         ]}}}
      CurveIndex=$System.NearCameraFade
     )
    )
    Scale RGB=&VectorFromFloat(Value=&FloatFromCurve(
      FloatCurve=DataInterface() {propertyValues {Curve {keys [
          <at 0 1 RCIM_Cubic>
          {arriveTangent=-1.1751, interpMode=RCIM_Cubic, leaveTangent=-1.1751, time=0.2, value=0.9}
          {arriveTangent=-1.125, interpMode=RCIM_Cubic, leaveTangent=-1.125, time=0.5641, value=0.3372}
          <at 1 0 RCIM_Cubic>
         ]}}}
      Scale Curve=10
     ))
    ScaleA=true
    ScaleRGB=true
   ) {enabled=false}
   /SetVariables(Particles.EnergyColor:LinearColor=&MultiplyLinearColor(Linear Color=$Particles.Color, Scale Factor=&MakeVector4_FromFloat(Float=30))) {isSetParameters=true}
   /SetVariables(Particles.GlowColor=&Multiply_LinearColorByFloat(Linear Color=$Particles.Color, Float=0.3)) {isSetParameters=true}
   &SkeletalMeshLocation(Bone Sampling Mode=::NewEnumerator4, MaskSpawnGroup=false, Orientation Sampling=::NewEnumerator1, Skeletal Mesh=$System.SK_Mesh)
  ]
  rendererClasses [&NiagaraSpriteRendererProperties]
  renderers [
   {rendererClass=&NiagaraSpriteRendererProperties}
   {rendererClass=&NiagaraSpriteRendererProperties, rendererIndex=1}
   {rendererClass=&NiagaraSpriteRendererProperties, rendererIndex=2}
  ]
 }
 Smoke {
  emitterUpdateScript [
   &EmitterState(
    Enable Distance Culling=true
    MaxDistance=13000
    MinDistance=1
    Module.bUseMaxDistance=true
    Module.bUseMinDistance=false
    Scalability Mode=::NewEnumerator1
   )
   &SpawnBurst_Instantaneous(Use Spawn Probability=false, UseLoopCountLimit=false)
   &SpawnBurst_Instantaneous(Spawn Time=0.1, Use Spawn Probability=false, UseLoopCountLimit=false) {module=SpawnBurst_Instantaneous001}
   &SpawnBurst_Instantaneous(Spawn Group=1, Use Spawn Probability=false, UseLoopCountLimit=false) {module=SpawnBurst_Instantaneous002}
  ]
  particleSpawnScript [
   /SetVariables(Particles.VisibilityTag=&SelectIntFromArray(Direct Array Index=$Emitter.SpawnGroup, Int Selection Array=DataInterface() {propertyValues {IntData [0,1]}})) {isSetParameters=true}
   /SetVariables(Particles.RandomFloat=&UniformRangedFloat(Minimum=0, Maximum=1)) {isSetParameters=true}
   /Niagara/Modules/Spawn/Initialization/InitializeParticle(
    Color=<rgba 0.0208 0.0208 0.0208 1>
    Lifetime=&UniformRangedFloat(Minimum=0.6, Maximum=0.8)
    Module.Write Color=true
    Module.Write Lifetime=true
    Module.Write Mass=true
    Module.Write Material Random=true
    Module.Write Position=false
    Module.Write Scale=false
    Module.Write SpriteRotation=true
    Module.Write SpriteSize=true
    Position=&SimulationPosition()
    Sprite Rotation=&UniformRangedFloat(Minimum=0, Maximum=360)
    Sprite Size=&Vector2DFromFloat(Value=&UniformRangedFloat(Minimum=26.1438, Maximum=39.2157))
   )
   &SkeletalMeshLocation(Bone Sampling Mode=::NewEnumerator4, MaskSpawnGroup=false, Orientation Sampling=::NewEnumerator1, Skeletal Mesh=$System.SK_Mesh)
   &ShapeLocation(
    MaskSpawnGroup=false
    Override Seed=false
    Shape Origin=$Particles.SkeletalMeshLocation.SampledPosition
    Sphere Radius=10
    UseOwnerScale=true
   )
   &AddVelocity(Velocity=<0 0 50>)
   &DynamicMaterialParameters(
    Index 0 Param 4=0.25
    Index 2 Param 2=&RandomRangeFloat(Minimum=0, Maximum=1)
    Module.Param0WriteEnabled=false
    Module.Param0WriteEnabled2=false
    Module.Param1WriteEnabled=false
    Module.Param1WriteEnabled2=true
    Module.Param2WriteEnabled=false
    Module.Param2WriteEnabled2=false
    Module.Param3WriteEnabled=true
    Module.Param3WriteEnabled2=false
    Write Parameter Index 2=true
   )
   /SetVariables(Particles.FlareColor=<rgba 0.8281 0.1048 0.0097 1>, Particles.LightColor=<rgba 0.8281 0.1048 0.0097 1>) {isSetParameters=true}
  ]
  particleUpdateScript [
   &ParticleState()
   &AccelerationForce(Acceleration=&UniformRangedVector(Minimum=<0 0 150>, Maximum=<0 0 650>)) {enabled=false}
   &Drag(Drag=3, Use Linear Drag=true, Use Rotational Drag=false)
   &DynamicMaterialParameters(
    Index 0 Param 1=&Add_Float(
     A=$Particles.RandomFloat
     B=&FloatFromCurve(FloatCurve=DataInterface() {propertyValues {Curve {keys [
          {arriveTangent=2.4104, time=0, value=1}
          {arriveTangent=0.2537, interpMode=RCIM_Cubic, leaveTangent=0.2537, tangentMode=RCTM_User, time=1, value=2}
         ]}}})
    )
    Index 0 Param 4=1
    Index 2 Param 1=&FloatFromCurve(FloatCurve=DataInterface() {propertyValues {Curve {keys [{interpMode=RCIM_Cubic, leaveTangent=0.9717, tangentMode=RCTM_Break, time=0.25, value=0.8}, <at 1 1>]}}})
    Index 2 Param 2=&FloatFromCurve(
     FloatCurve=DataInterface() {propertyValues {Curve {keys [<at 0 0.2 RCIM_Cubic>,<at 1 1.4 RCIM_Cubic>]}}}
     Scale Curve=&RandomRangeFloat(Minimum=0.2, Maximum=0.5)
    )
    Index 2 Param 3=&FloatFromCurve(FloatCurve=DataInterface() {propertyValues {Curve {keys [
         {interpMode=RCIM_Cubic, tangentMode=RCTM_Break, time=0, value=1}
         {interpMode=RCIM_Cubic, leaveTangent=11.1233, tangentMode=RCTM_Break, time=1, value=2}
        ]}}})
    Index 2 Param 4=&FloatFromCurve(
     FloatCurve=DataInterface() {propertyValues {Curve {keys [
         {arriveTangent=-0.4552, interpMode=RCIM_Cubic, leaveTangent=-0.4552, tangentMode=RCTM_User, time=0, value=1}
         {interpMode=RCIM_Cubic, leaveTangent=-1.2956, tangentMode=RCTM_Break, time=1, value=0.75}
        ]}}}
     Scale Curve=1.4
    )
    Module.Param0WriteEnabled=true
    Module.Param0WriteEnabled2=true
    Module.Param1WriteEnabled=false
    Module.Param1WriteEnabled2=true
    Module.Param2WriteEnabled=false
    Module.Param2WriteEnabled2=true
    Module.Param3WriteEnabled=false
    Module.Param3WriteEnabled2=true
    Write Parameter Index 2=true
   ) {module=DynamicMaterialParameters001}
   &ScaleSpriteSize(Scale Factor=&Vector2DFromFloat(Value=&FloatFromCurve(
      FloatCurve=DataInterface() {propertyValues {Curve {keys [
          {arriveTangent=3.358, interpMode=RCIM_Cubic, leaveTangent=3.358, tangentMode=RCTM_User, time=0, value=0.1773}
          {arriveTangent=0.8227, interpMode=RCIM_Cubic, leaveTangent=0.8227, time=0.1179, value=0.489}
          {interpMode=RCIM_Cubic, leaveTangent=0.0321, time=1, value=1}
         ]}}}
      Scale Curve=2
     )))
   &ScaleColor(
    Scale Alpha=&FloatFromCurve(FloatCurve=DataInterface() {propertyValues {Curve {keys [
         {arriveTangent=10.6894, interpMode=RCIM_Cubic, leaveTangent=10.6894, tangentMode=RCTM_User, time=0.0001, value=0.0123}
         {arriveTangent=0.6247, interpMode=RCIM_Cubic, leaveTangent=0.6247, time=0.15, value=1}
         {arriveTangent=-2.6727, interpMode=RCIM_Cubic, leaveTangent=-2.6727, tangentMode=RCTM_User, time=0.8071, value=0.5165}
         <at 1 0 RCIM_Cubic>
        ]}}})
    Scale RGB=&MakeVectorFromLinearColorRGB(Color=&ColorFromCurve(ColorCurve=DataInterface() {propertyValues {
        AlphaCurve {keys [<at 0 1 RCIM_Cubic>,<at 1 0 RCIM_Cubic>]}
        BlueCurve {keys [<at 0 1 RCIM_Cubic>,<at 0.2756 1 RCIM_Cubic>]}
        GreenCurve {keys [<at 0 4 RCIM_Cubic>,<at 0.2756 1 RCIM_Cubic>]}
        RedCurve {keys [<at 0 10 RCIM_Cubic>,<at 0.2756 1 RCIM_Cubic>]}
       }}))
    ScaleA=true
    ScaleRGB=true
   )
   &SpriteRotationRate(Rotation Rate=&Multiply_Float(
     A=&Lerp_Float(A=-150, B=150, Alpha=$Particles.MaterialRandom)
     B=&Lerp_Float(A=1, B=0, Alpha=&Power(a=$Particles.NormalizedAge, b=1))
    ))
   &CameraOffset(Camera Offset Amount=21)
   &SolveForcesAndVelocity(Clamp Velocity=false, Limit Acceleration=false)
   /SetVariables(Particles.Lifetime=&MaskFloat(
     Mask Source=1
     Mask Parameter=$Emitter.SpawnGroup
     Mask=0
     Masked Float=$Particles.Lifetime
     Passthrough Float=&Divide_Float(A=$Particles.Lifetime, B=3)
    )) {isSetParameters=true}
   /SetVariables(Particles.LightColor=<rgba 1 0.1162 0 1>) {enabled=false, isSetParameters=true}
   /SetVariables(
    Particles.FlareDyn:Vector4f=<0.2 1 0 0>
    Particles.FlareSize:Vector2f=<200 200>
    Particles.LightPos:NiagaraPosition=$Particles.SkeletalMeshLocation.SampledPosition
   ) {isSetParameters=true}
   &FN_Light_v2(Light Exponent=3, Light Radius=50, LightEnabled=true, Scale Color=1) {enabled=false}
   /SetVariables(Particles.FlareColor=&MultiplyLinearColor(
     Linear Color=$Particles.Initial.FlareColor
     Scale Factor=&MakeVector4(
      X=1
      Y=1
      Z=1
      W=&FloatFromCurve(FloatCurve=DataInterface() {propertyValues {Curve {keys [<at 0 0.3>,<at 1 0>]}}})
     )
    )) {isSetParameters=true}
  ]
  rendererClasses [&NiagaraSpriteRendererProperties, &NiagaraLightRendererProperties]
  renderers [
   {rendererClass=&NiagaraSpriteRendererProperties}
   {rendererClass=&NiagaraLightRendererProperties, rendererIndex=1}
   {rendererClass=&NiagaraSpriteRendererProperties, rendererIndex=2}
  ]
 }
}
summary {
 systemName=NS_VictoryHighway_HandFire
 emitters [
  {emitterName=Embers_Burst, simTarget=GPUComputeSim, rendererClasses [&NiagaraSpriteRendererProperties]}
  {emitterName=Fire_Burst, rendererClasses [&NiagaraSpriteRendererProperties]}
  {emitterName=Flametrail, rendererClasses [&NiagaraRibbonRendererProperties]}
  {emitterName=ParticleDecal, rendererClasses [&NiagaraSpriteRendererProperties]}
  {emitterName=Smoke, rendererClasses [&NiagaraSpriteRendererProperties, &NiagaraLightRendererProperties]}
  {emitterName=Drops, rendererClasses [&NiagaraSpriteRendererProperties]}
 ]
 userVariables [
  User.Color {type=LinearColor, description="", defaultValue {struct=LinearColor, value=<rgba 1 0 0.8389 1>}}
  User.Overall Scale {type=NiagaraFloat, defaultValue=0, description=""}
  User.SpawnRate {type=NiagaraFloat, defaultValue=1, description=""}
  User.UseCustomColor {type=NiagaraFloat, defaultValue=0, description=""}
  User.WaterBodyIndex {type=NiagaraFloat, defaultValue=0, description=""}
  User.WaterDepth {type=NiagaraFloat, defaultValue=0, description=""}
  User.WaterPlaneLocation {type=Vector3f, description="", defaultValue {struct=Vector3f, value=<0 0 0>}}
  User.WaterPlaneNormal {type=Vector3f, description="", defaultValue {struct=Vector3f, value=<0 0 0>}}
  User.WaveAttenuationFactor {type=NiagaraFloat, defaultValue=0, description=""}
  User.WaveTime {type=NiagaraFloat, defaultValue=0, description=""}
 ]
}
