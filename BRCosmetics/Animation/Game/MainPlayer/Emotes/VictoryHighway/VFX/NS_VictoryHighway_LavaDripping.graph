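# Niagara system for the Victory Highway player emote that drips molten lava from sampled points on a system-level skeletal mesh.
# Curve-driven, timed spawning feeds gravity-pulled ribbons split by partitioned ribbon IDs, plus falling droplet and spurt sprites.
# Glow sprites follow the mesh and dim by near-camera fade.
# Short-lived decals take XY from that mesh and Z from the owner skeleton, with random offsets.
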NiagaraSystem /BRCosmetics/Animation/Game/MainPlayer/Emotes/VictoryHighway/VFX/NS_VictoryHighway_LavaDripping
dependencies {
 usedDataInterfaces [
  &NiagaraDataInterfaceCurve
  &NiagaraDataInterfaceColorCurve
  &NiagaraDataInterfaceVector2DCurve
  &NiagaraDataInterfaceParticleRead
  &NiagaraDataInterfaceCamera
 ]
 usedDynamicInputs [/Niagara/DynamicInputs/Camera/CameraFade]
 usedModules [
  &self:SystemSpawnScript.SystemScriptSource.SystemScriptGraph.NiagaraNodeAssignment_0.SetVariables_#1
  &self:SystemSpawnScript.SystemScriptSource.SystemScriptGraph.NiagaraNodeAssignment_1.SetVariables_#2
  &SystemState
  &BlendSpaceHelper
 ]
}
emitters {
 Lava_Dripping {
  emitterUpdateScript [
   &EmitterState(Life Cycle Mode=::NewEnumerator1, Loop Behavior=::NewEnumerator1, Loop Duration=2.4, UseLoopDelay=false)
   &SpawnRate(
    SpawnRate=&FloatFromCurve(
     FloatCurve=DataInterface() {propertyValues {Curve {keys [{interpMode=RCIM_Cubic, tangentMode=RCTM_User, time=0.89, value=1}, {arriveTangent=-1.5813, tangentMode=RCTM_User, time=0.9, value=0}]}}}
     CurveIndex=$Emitter.NormalizedLoopAge
     Scale Curve=30
    )
    Use Spawn Probability=true
   )
  ]
  particleSpawnScript [
   &InitializeParticle(
    Color Mode=::NewEnumerator0
    Lifetime=5
    Ribbon Width=&FloatFromCurve(
     FloatCurve=DataInterface() {propertyValues {Curve {keys [<at 0 25>,<at 1 10>]}}}
     CurveIndex=$Emitter.LoopedAge
     Scale Curve=&RandomRangeFloat(Minimum=0.75, Maximum=1.75)
    )
    Ribbon Width Mode=::NewEnumerator1
    Sprite Size Mode=::NewEnumerator3
    Uniform Sprite Size=30
    UsePositionOffset=false
    Write Material Random=true
   )
   &SkeletalMeshLocation(
    Bone Sampling Mode=::NewEnumerator4
    MaskSpawnGroup=false
    Orientation Sampling=::NewEnumerator1
    Sampled Position Offset=<0 0 -10>
    Skeletal Mesh=$System.SK_Mesh
   )
   &AddVelocity(
    Cone Angle=15
    Cone Axis=<0 0 -1>
    Module.UseConeFalloff=false
    Module.UseVelDistribution=false
    Override Seed=false
    Velocity Mode=::NewEnumerator2
   )
   &PartitionParticles(
    Partition Count Mode:ENiagara_InfiniteLoopDuration=::NewEnumerator1
    Partition Duration:NiagaraFloat=&FloatFromCurve(FloatCurve=DataInterface() {propertyValues {Curve {keys [<at 0 0.5>,<at 1 0.1>]}}})
    Partition Mode:ENiagara_PartitionMode_v2=::NewEnumerator2
    Write Partition to Ribbon ID=::NewEnumerator0
   )
   &ShapeLocation(MaskSpawnGroup=false, Sphere Distribution=::NewEnumerator1, Sphere Radius=3, UseOwnerScale=true)
  ]
  particleUpdateScript [
   &ParticleState()
   &GravityForce(Gravity=<0 0 -350>)
   &SolveForcesAndVelocity(Clamp Velocity=false, Limit Acceleration=false)
   &ScaleRibbonWidth(Ribbon Width Scale=&FloatFromCurve(FloatCurve=DataInterface() {propertyValues {Curve {keys [
         {interpMode=RCIM_Cubic, leaveTangent=-0.3616, tangentMode=RCTM_Break, time=0, value=1}
         {arriveTangent=-1.0768, interpMode=RCIM_Cubic, tangentMode=RCTM_Break, time=1, value=0}
        ]}}}))
   &Color(Color=&ColorFromCurve(ColorCurve=DataInterface() {propertyValues {
       AlphaCurve {keys [<at 0 1 RCIM_Cubic>,<at 1 0 RCIM_Cubic>]}
       BlueCurve {keys [<at 0 0.0704 RCIM_Cubic>,<at 0.1377 0.0237>]}
       GreenCurve {keys [<at 0 0.2531 RCIM_Cubic>,<at 0.1377 0.0572>]}
       RedCurve {keys [<at 0 1 RCIM_Cubic>,<at 0.1377 0.5>]}
      }})) {module=Color001}
   &DynamicMaterialParameters(
    Index 0 Param 4=&FloatFromCurve(FloatCurve=DataInterface() {propertyValues {Curve {keys [<at 0 1>,<at 0.5 0>]}}})
    Module.Index0Float0Write=true
    Module.Index0Float1Write=true
    Module.Index0Float2Write=true
    Module.Index0Float3Write=true
   ) {module=DynamicMaterialParameters001}
  ]
  rendererClasses [&NiagaraRibbonRendererProperties]
  renderers [{rendererClass=&NiagaraRibbonRendererProperties}]
 }
 Minimal001 {
  emitterUpdateScript [
   &EmitterState(Life Cycle Mode=::NewEnumerator1, Loop Behavior=::NewEnumerator1, Loop Duration=3.4, UseLoopDelay=false)
   &SpawnRate(
    SpawnRate=&FloatFromCurve(
     FloatCurve=DataInterface() {propertyValues {Curve {keys [
         {arriveTangent=-0.8735, interpMode=RCIM_Cubic, leaveTangent=-0.8735, tangentMode=RCTM_User, time=0, value=1}
         {arriveTangent=-1.8342, interpMode=RCIM_Cubic, leaveTangent=-1.8342, tangentMode=RCTM_User, time=0.85, value=0}
        ]}}}
     CurveIndex=$Emitter.NormalizedLoopAge
     Scale Curve=30
    )
    Use Spawn Probability=false
   )
  ]
  particleSpawnScript [
   &InitializeParticle(
    Color=<rgba 3 1.1464 0.1805 1>
    Sprite Rotation Angle=0.5
    Sprite Rotation Mode=::NewEnumerator1
    Sprite Size=<2 6>
    Sprite Size Mode=::NewEnumerator4
    UsePositionOffset=false
    Write Material Random=true
   )
   &SkeletalMeshLocation(
    Bone Sampling Mode=::NewEnumerator4
    MaskSpawnGroup=false
    Orientation Sampling=::NewEnumerator1
    Sampled Position Offset=<0 0 -10>
    Skeletal Mesh=$System.SK_Mesh
   )
   &AddVelocity(Velocity=&RandomRangeVector(Minimum=<10 10 0>, Maximum=<-10 -10 1>)) {module=AddVelocity001}
  ]
  particleUpdateScript [
   &ParticleState()
   &GravityForce(Gravity=&RandomRangeVector(Minimum=<0 0 -400>, Maximum=<1 1 -650>))
   &SolveForcesAndVelocity(Clamp Velocity=false, Limit Acceleration=false)
   &ScaleSpriteSize(
    Non-Uniform Curve Scale=&Multiply_Vector2D_ByFloat(
     Float=&FloatFromCurve(
      FloatCurve=DataInterface() {propertyValues {Curve {keys [
          {arriveTangent=6.5967, interpMode=RCIM_Cubic, leaveTangent=6.5967, tangentMode=RCTM_User, time=0, value=0}
          {interpMode=RCIM_Cubic, tangentMode=RCTM_User, time=0.2271, value=1}
          {tangentMode=RCTM_User, time=1, value=0}
         ]}}}
      Scale Curve=&RandomRangeFloat(Minimum=0.9, Maximum=1.5)
     )
     Vector2D=<1 1>
    )
    Non-Uniform Curve Sprite Scale=DataInterface() {propertyValues {
      XCurve {keys [
        {arriveTangent=4.1785, interpMode=RCIM_Cubic, leaveTangent=4.1785, tangentMode=RCTM_User, time=0, value=0}
        {arriveTangent=1.1179, interpMode=RCIM_Cubic, leaveTangent=1.1179, tangentMode=RCTM_User, time=0.1, value=1.375}
        {arriveTangent=0.0224, interpMode=RCIM_Cubic, leaveTangent=0.0224, tangentMode=RCTM_User, time=1, value=1}
       ]}
      YCurve {keys [
        <at 0 0.2781 RCIM_Cubic>
        {arriveTangentWeight=0.5187, interpMode=RCIM_Cubic, leaveTangentWeight=0.2766, tangentMode=RCTM_User, time=0.863, value=2}
       ]}
     }}
    Scale Sprite Size Mode=::NewEnumerator3
   )
   &DynamicMaterialParameters(
    Index 0 Param 1=1
    Index 0 Param 2=1
    Index 0 Param 3=1.9518
    Index 0 Param 4=0.5547
    Module.Index0Float0Write=true
    Module.Index0Float1Write=true
    Module.Index0Float2Write=true
    Module.Index0Float3Write=true
    Module.Index1Float0Write=true
    Module.Index1Float1Write=true
    Module.Index1Float2Write=true
    Module.Index1Float3Write=true
    Module.Index2Float0Write=true
    Module.Index2Float1Write=true
    Module.Index2Float2Write=true
    Module.Index2Float3Write=true
    Module.Index3Float0Write=true
    Module.Index3Float1Write=true
    Module.Index3Float2Write=true
    Module.Index3Float3Write=true
    Write Parameter Index 1=::NewEnumerator1
    Write Parameter Index 2=::NewEnumerator1
    Write Parameter Index 3=::NewEnumerator1
   )
   &SpriteFacingAndAlignment(
    Sprite Alignment=&Lerp_Vector(
     A=&Multiply_Vector(A=Linked() {linkedVariable {name=Particles.Position, type=NiagaraPosition}}, B=<1 0 1>)
     B=<0 0 1>
     Alpha=&FloatFromCurve(
      FloatCurve=DataInterface() {propertyValues {Curve {keys [
          {interpMode=RCIM_Cubic, tangentMode=RCTM_User, time=0, value=0}
          {arriveTangent=2.5388, interpMode=RCIM_Cubic, leaveTangent=2.5388, tangentMode=RCTM_User, time=0.4178, value=1}
         ]}}}
      CurveIndex=$Emitter.Age
     )
    )
    WriteAlignment=true
    WriteFacing=false
   ) {module=SpriteFacingAndAlignment001}
  ]
  rendererClasses [&NiagaraSpriteRendererProperties]
  renderers [{rendererClass=&NiagaraSpriteRendererProperties}]
 }
 MustardDecals {
  emitterUpdateScript [
   &EmitterState(Inactive Response=::NewEnumerator1, Life Cycle Mode=::NewEnumerator1, Loop Behavior=::NewEnumerator1, Loop Duration=4.5, UseLoopDelay=false)
   &SpawnRate(
    SpawnRate=&FloatFromCurve(
     FloatCurve=DataInterface() {propertyValues {Curve {keys [<at 0 0>,<at 0.0534 1>,<at 0.4888 1>,<at 0.9 0>]}}}
     CurveIndex=$Emitter.NormalizedLoopAge
     Scale Curve=14
    )
    Use Spawn Probability=false
   )
   /SetVariables() {isSetParameters=true}
   /SetVariables(Emitter.MustardColA:LinearColor=<rgba 1 0.3743 0.0924 1>, Emitter.MustardColB:LinearColor=<rgba 0.4 0.0814 0.0278 1>) {isSetParameters=true}
  ]
  particleSpawnScript [
   &InitializeParticle(
    Color=&RandomRangeLinearColor(Minimum=<rgba 1 0.1981 0.1695 1>, Maximum=<rgba 1 0.427 0.0638 1>)
    Lifetime Max=1
    Lifetime Min=0.7
    Lifetime Mode=::NewEnumerator1
    Mass Max=1.5
    Mass Mode=::NewEnumerator3
    Sprite Rotation Mode=::NewEnumerator2
    Sprite Size Mode=::NewEnumerator1
    Uniform Sprite Size Max=2.5
    Uniform Sprite Size Min=1
    UsePositionOffset=false
    Write Material Random=true
   )
   &SkeletalMeshLocation(
    Bone Sampling Mode=::NewEnumerator4
    MaskSpawnGroup=false
    Orientation Sampling=::NewEnumerator1
    Sampled Position Offset=&RandomRangeVector(Minimum=<-50 -50 0>, Maximum=<50 50 1>)
    Sampled Position Offset Coordinate Space:ENiagaraSkelMeshTransforms=::NewEnumerator2
    Skeletal Mesh=$System.SK_Mesh
   ) {module=SkeletalMeshLocation001}
   &SkeletalMeshLocation(
    Bone Sampling Mode=::NewEnumerator4
    MaskSpawnGroup=false
    Orientation Sampling=::NewEnumerator1
    Position Sampling=::NewEnumerator2
    Sampled Position Offset Coordinate Space:ENiagaraSkelMeshTransforms=::NewEnumerator2
    Skeletal Mesh=$User.OwnerSkeleton
   )
   &DynamicMaterialParameters(
    Index 0 Param 1=2
    Module.Index0Float0Write=true
    Module.Index0Float1Write=true
    Module.Index0Float2Write=true
    Module.Index0Float3Write=true
   )
   /SetVariables(
    Particles.Rand:NiagaraFloat=&RandomRangeFloat(Minimum=4, Maximum=11)
    Particles.VisibilityTag=&RandomRangeInt(Minimum=0, Maximum=3)
   ) {isSetParameters=true}
   /SetVariables(Particles.Position=&Add_Vector(
     A=&MakeVector(
      X=&MakeFloatFromVector(Vector=Linked() {linkedVariable {name=Particles.SkeletalMeshLocation001.SampledPosition, type=NiagaraPosition}}, Channel=::NewEnumerator0)
      Y=&MakeFloatFromVector(Vector=Linked() {linkedVariable {name=Particles.SkeletalMeshLocation001.SampledPosition, type=NiagaraPosition}}, Channel=::NewEnumerator1)
      Z=&MakeFloatFromVector(Vector=Linked() {linkedVariable {name=Particles.SkeletalMeshLocation.SampledPosition, type=NiagaraPosition}}, Channel=::NewEnumerator2)
     )
     B=&RandomRangeVector(Minimum=<-3 -3 0>, Maximum=<3 3 0>)
    )) {isSetParameters=true}
  ]
  particleUpdateScript [
   &ParticleState()
   /SetVariables(Particles.DecalSize:Vector3f=&Multiply_Vector(
     A=&MakeVector(X=115, Y=$Particles.Rand, Z=$Particles.Rand)
     B=&VectorFromFloat(Value=&FloatFromCurve(FloatCurve=DataInterface() {propertyValues {Curve {keys [
           {arriveTangent=4.1785, interpMode=RCIM_Cubic, leaveTangent=4.1785, tangentMode=RCTM_User, time=0, value=0.5}
           {arriveTangent=1.1179, interpMode=RCIM_Cubic, leaveTangent=1.1179, tangentMode=RCTM_User, time=0.01, value=0.8}
           {arriveTangent=0.0224, interpMode=RCIM_Cubic, leaveTangent=0.0224, tangentMode=RCTM_User, time=0.05, value=1}
          ]}}}))
    )) {isSetParameters=true}
   /SetVariables(Particles.DecalFade:NiagaraFloat=&FloatFromCurve(FloatCurve=DataInterface() {propertyValues {Curve {keys [<at 0 1>,<at 1 0>]}}})) {isSetParameters=true}
   &ScaleColor(
    Scale Alpha=&FloatFromCurve(FloatCurve=DataInterface() {propertyValues {Curve {keys [<at 0 1>,<at 1 0>]}}})
    ScaleA=true
    ScaleRGB=true
   )
   /Niagara/Modules/AttributeReader/UpdateParticlesFromOtherEmitter(
    Attribute Reader=DataInterface() {propertyValues {EmitterBinding {bindingMode=Other, emitterName=OrbitBlobs_Trails002}}}
    Dynamic Material Parameter 1 Sampling:ENiagara_DirectReadSamplingMode=::NewEnumerator1
    Position Sampling:ENiagara_DirectReadSamplingMode=::NewEnumerator1
    Transform Data Option=1
   ) {enabled=false}
   /SetVariables(Particles.ConstructPos:Vector3f=&MakeVector(
     X=&MakeFloatFromVector4(Vector 4=$Output.UpdateParticlesFromOtherEmitter.DynamicMaterialParameter1, Channel=::NewEnumerator0)
     Y=&MakeFloatFromVector4(Vector 4=$Output.UpdateParticlesFromOtherEmitter.DynamicMaterialParameter1, Channel=::NewEnumerator1)
     Z=&MakeFloatFromVector4(Vector 4=$Output.UpdateParticlesFromOtherEmitter.DynamicMaterialParameter1, Channel=::NewEnumerator2)
    )) {enabled=false, isSetParameters=true}
   /SetVariables(Particles.Position=$Particles.SkeletalMeshLocation.SampledPosition) {enabled=false, isSetParameters=true}
  ]
  rendererClasses [&NiagaraDecalRendererProperties]
  renderers [
   {rendererClass=&NiagaraDecalRendererProperties, rendererIndex=0}
   {rendererClass=&NiagaraDecalRendererProperties, rendererIndex=1}
   {rendererClass=&NiagaraDecalRendererProperties, rendererIndex=2}
   {rendererClass=&NiagaraDecalRendererProperties, rendererIndex=3}
  ]
 }
 ParticleDecal {
  emitterUpdateScript [
   &EmitterState(
    Inactive Response=::NewEnumerator1
    Life Cycle Mode=::NewEnumerator1
    Loop Behavior=::NewEnumerator1
    Loop Duration Mode=::NewEnumerator1
    UseLoopDelay=false
   )
   &SpawnBurst_Instantaneous(Spawn Count=2, Use Spawn Probability=true, UseLoopCountLimit=false)
  ]
  particleSpawnScript [
   &PartitionParticles()
   &InitializeParticle(
    Color=<rgba 1 0.2549 0.0234 1>
    Lifetime=4
    Mesh Scale Mode=::NewEnumerator3
    Mesh Uniform Scale=10
    Sprite Rotation Mode=::NewEnumerator0
    Sprite Size=&MakeVector2D(
     X=&MaskFloat(Mask Source=1, Mask Parameter=$Particles.PartitionParticles.Partition, Mask=0, Masked Float=20, Passthrough Float=40)
     Y=40
    )
    Sprite Size Mode=::NewEnumerator4
    UsePositionOffset=false
    Write Material Random=true
   )
   &CameraOffset() {module=CameraOffset001}
   &DynamicMaterialParameters(
    Index 0 Param 1=1
    Index 0 Param 2=1
    Module.Index0Float0Write=true
    Module.Index0Float1Write=true
    Module.Index0Float2Write=true
    Module.Index0Float3Write=true
   )
   /SetVariables(Particles.EnergyDP:Vector4f=<2 1 8 1>) {isSetParameters=true}
   &SpriteFacingAndAlignment(
    Sprite Facing=&MaskVector(Mask Source=1, Mask Parameter=$Particles.PartitionParticles.Partition, Mask=0, Masked Vector=<1 0 0>, Passthrough Vector=<0 1 0>)
    WriteAlignment=false
    WriteFacing=true
   )
  ]
  particleUpdateScript [
   &ParticleState(Kill Particles When Lifetime Has Elapsed=false, Let Infinitely Lived Particles Die When Emitter Deactivates=true)
   &ScaleColor(
    Scale Mode=::NewEnumerator1
    Scale RGBA=&MakeVector4_FromLinearColor(LinearColor=<rgba 1 0.4583 0.0739 1>)
    ScaleRGBA=true
   ) {module=ScaleColor001}
   &ScaleColor(
    Scale Alpha=&Multiply_Float(
     A=1
     B=&FloatFromCurve(
      FloatCurve=DataInterface() {propertyValues {Curve {keys [
          {arriveTangent=1.5243, interpMode=RCIM_Cubic, leaveTangent=1.5243, tangentMode=RCTM_User, time=0, value=0.1}
          {arriveTangent=0.0515, interpMode=RCIM_Cubic, leaveTangent=0.0515, tangentMode=RCTM_User, time=1, value=1}
         ]}}}
      CurveIndex=$System.NearCameraFade
     )
    )
    Scale RGB=&VectorFromFloat(Value=5)
    ScaleA=true
    ScaleRGB=true
   ) {module=ScaleColor002}
   &ScaleColor(
    Scale Alpha=&Multiply_Float(
     A=&FloatFromCurve(FloatCurve=DataInterface() {propertyValues {Curve {keys [
          {arriveTangent=0.3051, interpMode=RCIM_Cubic, leaveTangent=0.3051, tangentMode=RCTM_User, time=0, value=1}
          {arriveTangent=0.3995, interpMode=RCIM_Cubic, leaveTangent=0.3995, tangentMode=RCTM_User, time=0.1408, value=0.9847}
          {arriveTangent=-0.0405, interpMode=RCIM_Cubic, leaveTangent=-0.0405, tangentMode=RCTM_User, time=1, value=0}
         ]}}})
     B=&FloatFromCurve(
      FloatCurve=DataInterface() {propertyValues {Curve {keys [
          {arriveTangent=1.5243, interpMode=RCIM_Cubic, leaveTangent=1.5243, tangentMode=RCTM_User, time=0, value=0.1}
          {arriveTangent=0.0515, interpMode=RCIM_Cubic, leaveTangent=0.0515, tangentMode=RCTM_User, time=1, value=1}
         ]}}}
      CurveIndex=$System.NearCameraFade
     )
    )
    Scale RGB=&VectorFromFloat(Value=&FloatFromCurve(
      FloatCurve=DataInterface() {propertyValues {Curve {keys [
          <at 0 1 RCIM_Cubic>
          {arriveTangent=-1.1751, interpMode=RCIM_Cubic, leaveTangent=-1.1751, time=0.2, value=0.9}
          {arriveTangent=-1.125, interpMode=RCIM_Cubic, leaveTangent=-1.125, time=0.5641, value=0.3372}
          <at 1 0 RCIM_Cubic>
         ]}}}
      Scale Curve=10
     ))
    ScaleA=true
    ScaleRGB=true
   ) {enabled=false}
   /SetVariables(Particles.EnergyColor:LinearColor=&MultiplyLinearColor(Linear Color=$Particles.Color, Scale Factor=&MakeVector4_FromFloat(Float=30))) {isSetParameters=true}
   /SetVariables(
    Particles.Decal Size:Vector2f=<10 10>
    Particles.GlowColor=&Multiply_LinearColorByFloat(
     Linear Color=$Particles.Color
     Float=&FloatFromCurve(
      FloatCurve=DataInterface() {propertyValues {Curve {keys [
          <at 0 1>
          {arriveTangent=-0.1495, interpMode=RCIM_Cubic, leaveTangent=-0.1495, tangentMode=RCTM_User, time=0.8, value=1}
          {arriveTangent=-4.8605, tangentMode=RCTM_User, time=1, value=0}
         ]}}}
      CurveIndex=$System.NormalizedLoopAge
      Scale Curve=0.5
     )
    )
   ) {isSetParameters=true}
   &SkeletalMeshLocation(Bone Sampling Mode=::NewEnumerator4, MaskSpawnGroup=false, Sampled Position Offset=<0 0 -5>, Skeletal Mesh=$System.SK_Mesh)
  ]
  rendererClasses [&NiagaraSpriteRendererProperties]
  renderers [
   {rendererClass=&NiagaraSpriteRendererProperties}
   {rendererClass=&NiagaraSpriteRendererProperties, rendererIndex=1}
   {rendererClass=&NiagaraSpriteRendererProperties, rendererIndex=2}
  ]
 }
 med_spurt {
  emitterUpdateScript [
   &EmitterState(Life Cycle Mode=::NewEnumerator1, Loop Behavior=::NewEnumerator1, Loop Duration=2.3, UseLoopDelay=false)
   &SpawnRate(
    SpawnRate=&FloatFromCurve(
     FloatCurve=DataInterface() {propertyValues {Curve {keys [
         {arriveTangent=-1.154, interpMode=RCIM_Cubic, leaveTangent=-1.154, tangentMode=RCTM_User, time=0.2134, value=1}
         {arriveTangent=-2.3841, interpMode=RCIM_Cubic, leaveTangent=-2.3841, tangentMode=RCTM_User, time=1, value=0}
        ]}}}
     CurveIndex=$Emitter.NormalizedLoopAge
     Scale Curve=30
    )
    Use Spawn Probability=false
   ) {module=SpawnRate002}
   &SpawnRate(Spawn Probability=0.5, SpawnRate=15, Use Spawn Probability=false) {enabled=false}
   &SpawnRate(
    SpawnRate=&FloatFromCurve(
     FloatCurve=DataInterface() {propertyValues {Curve {keys [
         <at 3.1 0.1 RCIM_Cubic>
         <at 6.8 0.975>
         {arriveTangent=-0.1591, interpMode=RCIM_Cubic, leaveTangent=-0.1591, time=8.5, value=0.5}
         {arriveTangent=-0.0741, interpMode=RCIM_Cubic, leaveTangent=-0.0741, time=10.1, value=0.45}
         {arriveTangent=-0.125, interpMode=RCIM_Cubic, leaveTangent=-0.125, time=11.2, value=0.3}
         {arriveTangent=-0.0405, interpMode=RCIM_Cubic, leaveTangent=-0.0405, tangentMode=RCTM_User, time=12.9, value=0.1}
         <at 15 0>
         {arriveTangent=-0.2711, interpMode=RCIM_Cubic, leaveTangent=-0.0075, tangentMode=RCTM_Break, time=15.5, value=0}
         <at 24 0>
        ]}}}
     CurveIndex=$Emitter.Age
     Scale Curve=15
    )
    Use Spawn Probability=false
   ) {enabled=false, module=SpawnRate001}
  ]
  particleSpawnScript [
   /SetVariables(Particles.SpawnIndex:NiagaraInt32=&ModuloInt(A=&ReturnParticleIDAttributes(), B=7)) {enabled=false, isSetParameters=true}
   &InitializeParticle(
    Color=<rgba 6 6 6 1>
    Lifetime Max=1.3
    Lifetime Min=0.4
    Lifetime Mode=::NewEnumerator1
    Sprite Rotation Mode=::NewEnumerator2
    Sprite Size Mode=::NewEnumerator1
    Sprite UV Mode=::NewEnumerator3
    UsePositionOffset=false
    Write Material Random=true
   )
   &SkeletalMeshLocation(
    Bone Sampling Mode=::NewEnumerator4
    MaskSpawnGroup=false
    Orientation Sampling=::NewEnumerator1
    Sampled Position Offset=<0 0 -10>
    Skeletal Mesh=$System.SK_Mesh
   )
   &AddVelocity(Velocity=&RandomRangeVector(Minimum=<0 0 -10>, Maximum=<1 1 -60>))
  ]
  particleUpdateScript [
   &ParticleState()
   &SpriteRotationRate(Rotation Rate=&RandomRangeFloat(Minimum=-20, Maximum=20))
   &ScaleSpriteSize(
    Uniform Curve Scale=&Multiply_Float(A=2, B=1)
    Uniform Curve Sprite Scale=DataInterface() {propertyValues {Curve {keys [
        {interpMode=RCIM_Cubic, leaveTangent=0.1321, tangentMode=RCTM_Break, time=0, value=0.2}
        {arriveTangent=2.9072, interpMode=RCIM_Cubic, leaveTangent=2.9072, tangentMode=RCTM_User, time=0.3, value=1.25}
        {tangentMode=RCTM_User, time=0.8, value=2.25}
       ]}}}
   )
   &GravityForce(Gravity=<0 0 -1200>)
   &ScaleVelocity(Velocity Scale=&VectorFromFloat(Value=$User.Overall Scale))
   &SolveForcesAndVelocity(Clamp Velocity=false, Limit Acceleration=false)
   &ScaleSpriteSizeBySpeed(Max Scale Factor=<1 2>)
   &ScaleColor(
    Scale Alpha=&FloatFromCurve(FloatCurve=DataInterface() {propertyValues {Curve {keys [
         <at 0 -1>
         {arriveTangent=-0.2728, interpMode=RCIM_Cubic, leaveTangent=-0.2728, tangentMode=RCTM_User, time=0.0002, value=0.5}
         {arriveTangent=0.1356, interpMode=RCIM_Cubic, leaveTangent=0.1356, tangentMode=RCTM_User, time=1, value=-1}
        ]}}})
    Scale RGB=&MakeVectorFromLinearColorRGB(Color=&MultiplyLinearColor(
      Linear Color=&ColorFromCurve(ColorCurve=DataInterface() {propertyValues {
         AlphaCurve {keys []}
         BlueCurve {keys [<at 0.1481 40>, {arriveTangent=-51.3191, interpMode=RCIM_Cubic, leaveTangent=-51.3191, time=0.198, value=12}, <at 0.9216 0.3>, <at 1 0.3>]}
         GreenCurve {keys [<at 0.1481 40>, {arriveTangent=-51.3191, interpMode=RCIM_Cubic, leaveTangent=-51.3191, time=0.198, value=12}, <at 0.9216 0.3>, <at 1 0.3>]}
         RedCurve {keys [<at 0.1481 40>, {arriveTangent=-51.3191, interpMode=RCIM_Cubic, leaveTangent=-51.3191, time=0.198, value=12}, <at 0.9216 0.3>, <at 1 0.3>]}
        }})
      Scale Factor=&MakeVector4_FromLinearColor(LinearColor=&LerpLinearColors(StartColor=<rgba 1.5 0.1323 0.0893 1>, EndColor=<rgba 0.875 0.2932 0.0021 1>, LerpFactor=$Particles.MaterialRandom))
     ))
    ScaleA=true
    ScaleRGB=true
   )
  ]
  rendererClasses [&NiagaraSpriteRendererProperties]
  renderers [{rendererClass=&NiagaraSpriteRendererProperties}]
 }
}
summary {
 systemName=NS_VictoryHighway_LavaDripping
 emitters [
  {emitterName=Lava_Dripping, rendererClasses [&NiagaraRibbonRendererProperties]}
  {emitterName=ParticleDecal, rendererClasses [&NiagaraSpriteRendererProperties]}
  {emitterName=med_spurt, rendererClasses [&NiagaraSpriteRendererProperties]}
  {emitterName=Minimal001, rendererClasses [&NiagaraSpriteRendererProperties]}
  {emitterName=MustardDecals, rendererClasses [&NiagaraDecalRendererProperties]}
 ]
 userVariables [
  User.bIsFrontEndPreview {type=NiagaraBool}
  User.Overall Scale {type=NiagaraFloat, defaultValue=0, description=""}
  User.OwnerSkeleton {
   type=NiagaraDataInterfaceSkeletalMesh
   defaultValue {
    struct=NiagaraExt_VariableValue_DataInterface
    value {dataInterface=&self:NiagaraDataInterfaceSkeletalMesh_1}
   }
  }
 ]
}
